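# Physics asset for the GoodMood female medium player body, set up for secondary motion on its straps and a pouch chain.
# Front and back strap capsules form limited-swing chains from base anchors, joined at their tips by reciprocal constraint pairs into a loop on each side.
# A short chain links base_chainTop to base_pouch.
# Unconstrained capsules on the pelvis, thighs, right calf and heart bones supply body collision shapes.
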PhysicsAsset /BRCosmetics/Characters/Player/Female/Medium/Bodies/F_MED_GoodMood/Meshes/F_MED_GoodMood_Physics
bodies {
 base_backStrap_l [{shapeType=Capsule, length=0.505, radius=0.505, shapeName=base_backStrap_l_capsule, rotation=<rot 0 -90 0>}]
 base_backStrap_r [{shapeType=Capsule, length=0.505, radius=0.505, shapeName=base_backStrap_r_capsule, rotation=<rot 0 -90 0>}]
 base_chainTop [{shapeType=Capsule, length=0, radius=1, shapeName=base_chainTop_capsule}]
 base_frontStrap_l [{shapeType=Capsule, length=0.505, radius=0.505, shapeName=base_frontStrap_l_capsule, rotation=<rot 0 -90 0>}]
 base_frontStrap_r [{shapeType=Capsule, length=0.505, radius=0.505, shapeName=base_frontStrap_r_capsule, rotation=<rot 0 -90 0>}]
 base_heart_l [
  {
   shapeType=Capsule
   length=28.5
   radius=7.505
   shapeName=base_heart_l_capsule
   center=<-9.8233 -2.6505 -13.6309>
   rotation=<rot 0 4.7634 89.9996>
  }
  {
   shapeType=Capsule
   length=28.5
   radius=7.505
   shapeName=base_heart_l_capsule
   center=<-7.7804 -2.4801 -25.29>
   rotation=<rot 0 4.7634 89.9996>
  }
 ]
 base_heart_r [
  {
   shapeType=Capsule
   length=29.5
   radius=7.505
   shapeName=base_heart_l_capsule
   center=<-11.6197 3.1967 -32.2081>
   rotation=<rot 0.0001 -2.4658 -89.9992>
  }
  {
   shapeType=Capsule
   length=29.5
   radius=7.505
   shapeName=base_heart_r_capsule
   center=<-12.2121 3.3184 -19.3538>
   rotation=<rot 0.0001 -2.4658 -89.9992>
  }
 ]
 base_pouch [{
   shapeType=Capsule
   length=9.5
   radius=4.1
   shapeName=base_pouch_capsule
   center=<2.6039 -0.1795 -6.6851>
   rotation=<rot -3.789 0.0044 0.0537>
  }]
 calf_r [{
   shapeType=Capsule
   length=44.5
   radius=6.5
   shapeName=calf_twist_02_l_capsule
   center=<21.2591 1.2379 -0.0848>
   rotation=<rot 88.1572 -178.2565 175.7191>
  }]
 dyn_backStrap_l_01 [{
   shapeType=Capsule
   length=1.5
   radius=1.25
   shapeName=dyn_backStrap_l_01_capsule
   center=<3.1034 0.0541 -0.1264>
   rotation=<rot 0 -86.6081 -89.7324>
  }]
 dyn_backStrap_l_02 [{
   shapeType=Capsule
   length=3.7161
   radius=1
   shapeName=dyn_backStrap_l_02_capsule
   center=<3.3862 -0.146 0.2323>
   rotation=<rot 0 89.658 -87.246>
  }]
 dyn_backStrap_l_03 [{
   shapeType=Capsule
   length=4.5
   radius=1.15
   shapeName=dyn_backStrap_l_03_capsule
   center=<3.2973 0.3219 -0.0162>
   rotation=<rot 0.06 -80.5503 -87.576>
  }]
 dyn_backStrap_l_04 [{
   shapeType=Capsule
   length=3.953
   radius=1.0237
   shapeName=dyn_backStrap_l_04_capsule
   center=<4.4311 0.2492 -0.518>
   rotation=<rot 0 -86.0215 -83.3488>
  }]
 dyn_backStrap_r_01 [{
   shapeType=Capsule
   length=2
   radius=1.5
   shapeName=dyn_backStrap_r_01_capsule
   center=<2.3221 0.1571 -0.1926>
   rotation=<rot 0 -88.4811 -88.9414>
  }]
 dyn_backStrap_r_02 [{
   shapeType=Capsule
   length=6.7161
   radius=1.25
   shapeName=dyn_backStrap_r_02_capsule
   center=<4.0772 -0.152 -0.1024>
   rotation=<rot 0 88.1916 -87.859>
  }]
 dyn_backStrap_r_03 [{
   shapeType=Capsule
   length=11.0153
   radius=1.4
   shapeName=dyn_backStrap_r_03_capsule
   center=<8.9706 -0.1421 0.4445>
   rotation=<rot 0.0672 84.4422 -88.1729>
  }]
 dyn_chain [{shapeType=Capsule, length=6, radius=2.25, shapeName=dyn_chain_capsule, center=<0 -0.3043 2.6972>, rotation=<rot 0 -0.0001 87.1689>}]
 dyn_chainTop [{
   shapeType=Capsule
   length=1.5
   radius=1.25
   shapeName=base_chainTop_capsule
   center=<2.9465 -0.0005 0.1305>
   rotation=<rot -83.5753 -179.971 179.9607>
  }]
 dyn_frontStrap_l_01 [{
   shapeType=Capsule
   length=2
   radius=1
   shapeName=dyn_frontStrap_l_01_capsule
   center=<2.74 -0.011 0.5954>
   rotation=<rot -1.3419 -88.8823 -88.2206>
  }]
 dyn_frontStrap_l_02 [{
   shapeType=Capsule
   length=3.5
   radius=1.12
   shapeName=dyn_frontStrap_l_02_capsule
   center=<3.3071 -0.008 0.4045>
   rotation=<rot -0.0386 86.5597 -89.7921>
  }]
 dyn_frontStrap_l_03 [{
   shapeType=Capsule
   length=4
   radius=1.25
   shapeName=dyn_frontStrap_l_03_capsule
   center=<3.6461 -0.1439 -0.3551>
   rotation=<rot -0.793 -92.735 -87.6956>
  }]
 dyn_frontStrap_l_04 [{
   shapeType=Capsule
   length=3
   radius=1.2
   shapeName=dyn_frontStrap_l_03_capsule
   center=<3.0536 0.0064 -0.2657>
   rotation=<rot -1.0726 -92.0091 -81.0779>
  }]
 dyn_frontStrap_l_05 [{
   shapeType=Capsule
   length=2
   radius=1.15
   shapeName=dyn_frontStrap_l_03_capsule
   center=<1.6596 -0.3065 0.1367>
   rotation=<rot -85.2927 0.0008 -10.4302>
  }]
 dyn_frontStrap_r_01 [{
   shapeType=Capsule
   length=2
   radius=1
   shapeName=dyn_frontStrap_r_01_capsule
   center=<2.8595 0.1488 0.3638>
   rotation=<rot 0.4449 -89.1077 -90.8472>
  }]
 dyn_frontStrap_r_02 [{
   shapeType=Capsule
   length=5.3409
   radius=1.15
   shapeName=dyn_frontStrap_r_02_capsule
   center=<4.7182 -0.1137 0.3136>
   rotation=<rot 0 88.7789 -87.5955>
  }]
 dyn_frontStrap_r_03 [{
   shapeType=Capsule
   length=12.6162
   radius=1.15
   shapeName=dyn_frontStrap_r_03_capsule
   center=<8.5604 1.6364 -1.1207>
   rotation=<rot 0.2028 -74.285 -78.5823>
  }]
 pelvis [{shapeType=Capsule, length=9.5396, radius=8.9618, shapeName=pelvis_capsule, center=<0.1596 -1.4919 -0.367>, rotation=<rot 1.5237 -90 2.2971>}]
 thigh_l [{
   shapeType=Capsule
   length=31.5
   radius=6.25
   shapeName=calf_twist_02_l_capsule
   center=<-19.298 0.5211 0.7885>
   rotation=<rot 88.3226 -2.5047 -179.643>
  }]
 thigh_r [{
   shapeType=Capsule
   length=33.5
   radius=6.25
   shapeName=calf_twist_02_l_capsule
   center=<20.7475 -0.5936 -0.831>
   rotation=<rot 88.3226 -2.5047 0.357>
  }]
 thigh_twist_01_r [{shapeType=Capsule, length=9.1226, radius=6.9523, shapeName=thigh_twist_01_r_capsule, center=<-11.6985 0.5491 0>, rotation=<rot 90 0 7.1809>}]
}
constraints [
 {bone1Name=dyn_chain, bone2Name=dyn_chainTop, swing1LimitDegrees=5, swing2LimitDegrees=20, twistLimitDegrees=2.5}
 {bone1Name=base_pouch, bone2Name=dyn_chain, swing1LimitDegrees=2.5, swing2LimitDegrees=15, twistLimitDegrees=2.5}
 {
  bone1Name=dyn_frontStrap_r_03
  bone2Name=dyn_frontStrap_r_02
  swing1LimitDegrees=40
  swing2LimitDegrees=40
  twistLimitDegrees=1
  twistMotion=Locked
 }
 {
  bone1Name=dyn_frontStrap_r_02
  bone2Name=dyn_frontStrap_r_01
  swing1LimitDegrees=60
  swing2LimitDegrees=17
  twistLimitDegrees=1
  twistMotion=Locked
 }
 {bone1Name=dyn_frontStrap_r_01, bone2Name=base_frontStrap_r, swing1LimitDegrees=40, swing2LimitDegrees=10, twistLimitDegrees=5}
 {bone1Name=dyn_backStrap_r_03, bone2Name=dyn_backStrap_r_02, swing1LimitDegrees=80, swing2LimitDegrees=30, twistLimitDegrees=5}
 {bone1Name=dyn_backStrap_r_02, bone2Name=dyn_backStrap_r_01, swing1LimitDegrees=60, swing2LimitDegrees=20, twistLimitDegrees=10}
 {
  bone1Name=dyn_backStrap_r_01
  bone2Name=base_backStrap_r
  swing1LimitDegrees=40
  swing2LimitDegrees=5
  twistLimitDegrees=1
  twistMotion=Locked
 }
 {bone1Name=dyn_frontStrap_l_03, bone2Name=dyn_frontStrap_l_02, swing1LimitDegrees=20, swing2LimitDegrees=20, twistLimitDegrees=10}
 {bone1Name=dyn_frontStrap_l_02, bone2Name=dyn_frontStrap_l_01, swing1LimitDegrees=27, swing2LimitDegrees=45, twistLimitDegrees=15}
 {
  bone1Name=dyn_frontStrap_l_01
  bone2Name=base_frontStrap_l
  swing1LimitDegrees=40
  swing2LimitDegrees=10
  twistLimitDegrees=1
  twistMotion=Locked
 }
 {bone1Name=dyn_backStrap_l_03, bone2Name=dyn_backStrap_l_02, swing1LimitDegrees=80, swing2LimitDegrees=30, twistLimitDegrees=5}
 {bone1Name=dyn_backStrap_l_01, bone2Name=base_backStrap_l, swing1LimitDegrees=45, swing2LimitDegrees=20, twistLimitDegrees=5}
 {bone1Name=dyn_backStrap_l_02, bone2Name=dyn_backStrap_l_01, swing1LimitDegrees=60, swing2LimitDegrees=20, twistLimitDegrees=5}
 {bone1Name=dyn_backStrap_r_03, bone2Name=dyn_frontStrap_r_03, swing1LimitDegrees=45, swing2LimitDegrees=10, twistLimitDegrees=5}
 {bone1Name=dyn_frontStrap_r_03, bone2Name=dyn_backStrap_r_03, swing1LimitDegrees=45, swing2LimitDegrees=10, twistLimitDegrees=5}
 {bone1Name=dyn_chainTop, bone2Name=base_chainTop, swing1LimitDegrees=2.5, swing2LimitDegrees=20, twistLimitDegrees=2.5}
 {bone1Name=dyn_frontStrap_l_04, bone2Name=dyn_frontStrap_l_03, swing1LimitDegrees=20, swing2LimitDegrees=10, twistLimitDegrees=15}
 {bone1Name=dyn_frontStrap_l_05, bone2Name=dyn_frontStrap_l_04, swing1LimitDegrees=45, swing2LimitDegrees=10, twistLimitDegrees=10}
 {bone1Name=dyn_backStrap_l_04, bone2Name=dyn_backStrap_l_03, swing1LimitDegrees=35, swing2LimitDegrees=10, twistLimitDegrees=10}
 {bone1Name=dyn_backStrap_l_04, bone2Name=dyn_frontStrap_l_05, swing1LimitDegrees=45, swing2LimitDegrees=10, twistLimitDegrees=5}
 {bone1Name=dyn_frontStrap_l_05, bone2Name=dyn_backStrap_l_04, swing1LimitDegrees=45, swing2LimitDegrees=10, twistLimitDegrees=10}
]
