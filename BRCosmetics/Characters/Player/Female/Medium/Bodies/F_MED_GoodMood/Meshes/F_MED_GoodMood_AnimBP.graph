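# Post-copy AnimBlueprint for the GoodMood female medium body, adding secondary motion to its scarf, straps, pouch, zippers and pants.
# The parent pose runs through a deform Control Rig, bone constraints and AnimDynamics, bypassing scarf simulation when back bling is worn.
# Additive bone offsets correct crouch, skydive and front-end poses before a pelvis-space RigidBody solve.
# The EventGraph updates gravity override, crouch/skydive alphas and back-bling detection from attach sockets.
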AnimBlueprint /BRCosmetics/Characters/Player/Female/Medium/Bodies/F_MED_GoodMood/Meshes/F_MED_GoodMood_AnimBP
target_skeleton=&/F_MED_GoodMood_Skeleton
graphs [
 AnimGraph {nodes [
   n1 = Get Is Sky Diving() -> (Is Sky Diving:double)
   n2 = "Transform (Modify) Bone - Bone: base_frontPant_l"(
    ComponentPose=Local To Component(LocalPose=Use cached pose 'CLogToSDLog'())
    Alpha=@n1
    Translation="0,0,0"
    Rotation="5,2.5,0"
    Scale="1,1,1"
   ) {
    rotationMode=BMM_Additive
    rotationSpace=BCS_BoneSpace
    alphaBoolBlend {blendOption=Linear}
    boneToModify {boneName=base_frontPant_l}
    rotation {pitch=5, yaw=2.5}
   }
   n3 = "Transform (Modify) Bone - Bone: base_frontPant_r"(ComponentPose=@n2, Alpha=@n1, Translation="0,0,0", Rotation="5,2.5,2.5", Scale="1,1,1") {
    rotationMode=BMM_Additive
    rotationSpace=BCS_BoneSpace
    alphaBoolBlend {blendOption=Linear}
    boneToModify {boneName=base_frontPant_r}
    rotation=<rot 5 2.5 2.5>
   }
   n4 = "Transform (Modify) Bone - Bone: base_frontPant_r"(ComponentPose=@n3, Alpha=1, Translation="-1,0,0", Rotation="2.5,2.5,2.5", Scale="1,1,1") {
    rotationMode=BMM_Additive
    rotationSpace=BCS_BoneSpace
    translationMode=BMM_Additive
    translationSpace=BCS_BoneSpace
    alphaBoolBlend {blendOption=Linear}
    boneToModify {boneName=base_frontPant_r}
    rotation=<rot 2.5 2.5 2.5>
    translation {x=-1}
   }
   n5 = "Transform (Modify) Bone - Bone: base_frontPant_l"(ComponentPose=@n4, Alpha=1, Translation="-1,0,0", Rotation="2.5,2.5,2.5", Scale="1,1,1") {
    rotationMode=BMM_Additive
    rotationSpace=BCS_BoneSpace
    translationMode=BMM_Additive
    translationSpace=BCS_BoneSpace
    alphaBoolBlend {blendOption=Linear}
    boneToModify {boneName=base_frontPant_l}
    rotation=<rot 2.5 2.5 2.5>
    translation {x=-1}
   }
   n6 = "Transform (Modify) Bone - Bone: base_chainTop"(ComponentPose=@n5, Alpha=1, Translation="0.028,0.006,-0.080", Rotation="20,-10,0", Scale="1,1,1") {
    rotationMode=BMM_Additive
    rotationSpace=BCS_BoneSpace
    translationMode=BMM_Additive
    translationSpace=BCS_BoneSpace
    alphaBoolBlend {blendOption=Linear}
    boneToModify {boneName=base_chainTop}
    rotation {pitch=20, yaw=-10}
    translation=<0.028 0.006 -0.08>
   }
   n7 = "Transform (Modify) Bone - Bone: base_chainTop"!(ComponentPose=@n6, bAlphaBoolEnabled:bool=Get bIsInFrontEnd(), Translation="-1,0,-0.5", Rotation="0,10,0", Scale="1,1,1") -> (Pose) {
    alphaInputType=Bool
    rotationMode=BMM_Additive
    rotationSpace=BCS_BoneSpace
    translationMode=BMM_Additive
    translationSpace=BCS_BoneSpace
    alphaBoolBlend {blendOption=Linear}
    boneToModify {boneName=base_chainTop}
    rotation {yaw=10}
    translation {x=-1, z=-0.5}
   }
   n8 = RigidBody!(ComponentPose=@n7, Alpha, OverrideWorldGravity:Vector=Get GravityOverride(), ExternalForce) -> (Pose) {
    bOverrideWorldGravity=true
    overridePhysicsAsset=&/F_MED_GoodMood_Physics
    simulationSpace=BaseBoneSpace
    alphaBoolBlend {blendOption=Linear}
    baseBoneRef {boneName=pelvis}
    componentAppliedLinearAccClamp=<3000 3000 3000>
    componentLinearAccScale=<0.3 0.2 0.2>
    componentLinearVelScale=<0.3 0.2 0.2>
    simSpaceSettings {
     maxAngularAcceleration=1000
     maxAngularVelocity=3
     maxLinearAcceleration=700
     maxLinearVelocity=500
     velocityScaleZ=0
     worldAlpha=0.5
     externalLinearDragV=<0.2 0.2>
    }
   }
   Output Pose(Result=Component To Local(ComponentPose=@n8)) {layerGroup=None}
   n9 = Control Rig(Source=Copy Pose From Mesh() {bUseAttachedParent=true}) {
    alphaBoolBlend {blendOption=Linear}
    controlRigAssetReference {blueprintRigClass=&/F_MED_GoodMood_DeformRig.F_MED_GoodMood_DeformRig_C}
   }
   n10 = "Constraint - base_scarfTail (PPPPP)"!(
    ComponentPose=Local To Component(LocalPose=@n9)
    Alpha=1
    ConstraintWeights_0=0.7729
    ConstraintWeights_1:float=0.1948
    ConstraintWeights_2:float=0.0188
    ConstraintWeights_3:float=0.0134
    ConstraintWeights_4:float=0.0001
   ) -> (Pose) {
    alphaBoolBlend {blendOption=Linear}
    boneToModify {boneName=base_scarfTail}
    constraintSetup [
     {targetBone {boneName=clavicle_l}}
     {targetBone {boneName=spine_05}}
     {targetBone {boneName=deform_upperarm_up_l}}
     {targetBone {boneName=upperarm_twist_01_l}}
     {targetBone {boneName=deform_upperarm_fr_l}}
    ]
    constraintWeights [1,1,1,1,1]
   }
   n11 = "Constraint - base_heart_l (PPPPP)"!(
    ComponentPose=@n10
    Alpha=1
    ConstraintWeights_0=0.4534
    ConstraintWeights_1:float=0.4364
    ConstraintWeights_2:float=0.0717
    ConstraintWeights_3:float=0.0235
    ConstraintWeights_4:float=0.0149
   ) -> (Pose) {
    alphaBoolBlend {blendOption=Linear}
    boneToModify {boneName=base_heart_l}
    constraintSetup [
     {targetBone {boneName=pelvis}}
     {targetBone {boneName=thigh_twist_01_l}}
     {targetBone {boneName=deform_groin_l}}
     {targetBone {boneName=deform_glute_l}}
     {targetBone {boneName=thigh_l}}
    ]
    constraintWeights [1,1,1,1,1]
   }
   n12 = "Constraint - base_heart_r (PPPPP)"!(
    ComponentPose=@n11
    Alpha=1
    ConstraintWeights_0=0.8676
    ConstraintWeights_1:float=0.0746
    ConstraintWeights_2:float=0.0343
    ConstraintWeights_3:float=0.0234
    ConstraintWeights_4:float=0.0001
   ) -> (Pose) {
    alphaBoolBlend {blendOption=Linear}
    boneToModify {boneName=base_heart_r}
    constraintSetup [
     {targetBone {boneName=pelvis}}
     {targetBone {boneName=thigh_twist_01_r}}
     {targetBone {boneName=spine_02}}
     {targetBone {boneName=deform_groin_r}}
     {targetBone {boneName=deform_glute_r}}
    ]
    constraintWeights [1,1,1,1,1]
   }
   n13 = "Constraint - base_chainTop (PPPPP)"!(
    ComponentPose=@n12
    Alpha=1
    ConstraintWeights_0=0.7907
    ConstraintWeights_1:float=0.1609
    ConstraintWeights_2:float=0.0317
    ConstraintWeights_3:float=0.0093
    ConstraintWeights_4:float=0.0074
   ) -> (Pose) {
    alphaBoolBlend {blendOption=Linear}
    boneToModify {boneName=base_chainTop}
    constraintSetup [
     {targetBone {boneName=pelvis}}
     {targetBone {boneName=thigh_twist_01_r}}
     {targetBone {boneName=deform_groin_r}}
     {targetBone {boneName=deform_glute_r}}
     {targetBone {boneName=spine_02}}
    ]
    constraintWeights [1,1,1,1,1]
   }
   n14 = "Constraint - base_frontStrap_l (PPPP)"!(
    ComponentPose=@n13
    Alpha=1
    ConstraintWeights_0=0.5184
    ConstraintWeights_1:float=0.1887
    ConstraintWeights_2:float=0.182
    ConstraintWeights_3:float=0.1108
   ) -> (Pose) {
    alphaBoolBlend {blendOption=Linear}
    boneToModify {boneName=base_frontStrap_l}
    constraintSetup [
     {targetBone {boneName=thigh_twist_01_l}}
     {targetBone {boneName=deform_groin_l}}
     {targetBone {boneName=pelvis}}
     {targetBone {boneName=thigh_l}}
    ]
    constraintWeights [1,1,1,1]
   }
   n15 = "Constraint - base_backStrap_l (PPPP)"!(
    ComponentPose=@n14
    Alpha=2
    ConstraintWeights_0=0.4787
    ConstraintWeights_1:float=0.3533
    ConstraintWeights_2:float=0.158
    ConstraintWeights_3:float=0.0101
   ) -> (Pose) {
    alphaBoolBlend {blendOption=Linear}
    boneToModify {boneName=base_backStrap_l}
    constraintSetup [
     {targetBone {boneName=deform_glute_l}}
     {targetBone {boneName=thigh_twist_01_l}}
     {targetBone {boneName=pelvis}}
     {targetBone {boneName=thigh_l}}
    ]
    constraintWeights [1,1,1,1]
   }
   n16 = "Constraint - base_frontStrap_r (PPPP)"!(
    ComponentPose=@n15
    Alpha=1
    ConstraintWeights_0=0.5184
    ConstraintWeights_1:float=0.1887
    ConstraintWeights_2:float=0.182
    ConstraintWeights_3:float=0.1108
   ) -> (Pose) {
    alphaBoolBlend {blendOption=Linear}
    boneToModify {boneName=base_frontStrap_r}
    constraintSetup [
     {targetBone {boneName=thigh_twist_01_r}}
     {targetBone {boneName=deform_groin_r}}
     {targetBone {boneName=pelvis}}
     {targetBone {boneName=thigh_r}}
    ]
    constraintWeights [1,1,1,1]
   }
   n17 = "Constraint - base_backStrap_r (PPPP)"!(
    ComponentPose=@n16
    Alpha=1
    ConstraintWeights_0=0.4787
    ConstraintWeights_1:float=0.3533
    ConstraintWeights_2:float=0.158
    ConstraintWeights_3:float=0.0101
   ) -> (Pose) {
    alphaBoolBlend {blendOption=Linear}
    boneToModify {boneName=base_backStrap_r}
    constraintSetup [
     {targetBone {boneName=deform_glute_r}}
     {targetBone {boneName=thigh_twist_01_r}}
     {targetBone {boneName=pelvis}}
     {targetBone {boneName=thigh_r}}
    ]
    constraintWeights [1,1,1,1]
   }
   Save cached pose 'conToScarfAnimDyn'(Pose=Component To Local(ComponentPose=@n17)) {cacheName=conToScarfAnimDyn}
   n18 = Get GravityOverride()
   n19 = Anim Dynamics - Bone: base_scarfTail!(
    ComponentPose=Local To Component(LocalPose=Use cached pose 'conToScarfAnimDyn'())
    Alpha=0.6
    GravityOverride:Vector=@n18
    ExternalForce="0,0,0"
   ) -> (Pose) {
    angularDampingOverride=1
    angularSpringConstant=20000
    bAngularSpring=true
    bOverrideAngularDamping=true
    bUseGravityOverride=true
    gravityScale=0.35
    simulationSpace=RootRelative
    alphaBoolBlend {blendOption=Linear}
    boundBone {boneName=base_scarfTail}
    componentLinearAccScale=<1 1 1>
    componentLinearVelScale=<1 1 1>
    gravityOverride {z=-980}
    physicsBodyDefinitions [{
      boundBone {boneName=base_scarfTail}
      boxExtents=<20 30 30>
      constraintSetup {angularLimitsMax=<15 1 0>, angularLimitsMin=<-15 -30 -20>, angularTarget=<1 0 0>}
      localJointOffset=<4 0 0>
     }]
   }
   n20 = Anim Dynamics - Bone: dyn_scarfTail_01!(ComponentPose=@n19, Alpha=0.6, GravityOverride:Vector=@n18, ExternalForce="0,0,0") -> (Pose) {
    angularDampingOverride=1
    angularSpringConstant=20000
    bAngularSpring=true
    bOverrideAngularDamping=true
    bUseGravityOverride=true
    gravityScale=0.35
    simulationSpace=RootRelative
    alphaBoolBlend {blendOption=Linear}
    boundBone {boneName=dyn_scarfTail_01}
    componentLinearAccScale=<1 1 1>
    componentLinearVelScale=<1 1 1>
    gravityOverride {z=-980}
    physicsBodyDefinitions [{
      boundBone {boneName=dyn_scarfTail_01}
      boxExtents=<20 30 30>
      constraintSetup {angularLimitsMax=<15 1 0>, angularLimitsMin=<-15 -35 -20>, angularTarget=<1 0 0>}
      localJointOffset=<4 0 0>
     }]
   }
   n21 = Anim Dynamics - Bone: dyn_scarfTail_02!(ComponentPose=@n20, Alpha=0.75, GravityOverride:Vector=@n18, ExternalForce="0,0,0") -> (Pose) {
    angularDampingOverride=1
    angularSpringConstant=20000
    bAngularSpring=true
    bOverrideAngularDamping=true
    bUseGravityOverride=true
    gravityScale=0.35
    simulationSpace=RootRelative
    alphaBoolBlend {blendOption=Linear}
    boundBone {boneName=dyn_scarfTail_02}
    componentLinearAccScale=<1 1 1>
    componentLinearVelScale=<1 1 1>
    gravityOverride {z=-980}
    physicsBodyDefinitions [{
      boundBone {boneName=dyn_scarfTail_02}
      boxExtents=<20 30 30>
      constraintSetup {angularLimitsMax=<0 1 0>, angularLimitsMin=<-15 -40 -20>, angularTarget=<1 0 0>}
      localJointOffset=<4 0 0>
     }]
   }
   n22 = Anim Dynamics - Bone: dyn_scarfTail_03!(ComponentPose=@n21, Alpha=0.85, GravityOverride:Vector=@n18, ExternalForce="0,0,0") -> (Pose) {
    angularDampingOverride=1
    angularSpringConstant=20000
    bAngularSpring=true
    bOverrideAngularDamping=true
    bUseGravityOverride=true
    gravityScale=0.35
    simulationSpace=RootRelative
    alphaBoolBlend {blendOption=Linear}
    boundBone {boneName=dyn_scarfTail_03}
    componentLinearAccScale=<1 1 1>
    componentLinearVelScale=<1 1 1>
    gravityOverride {z=-980}
    physicsBodyDefinitions [{
      boundBone {boneName=dyn_scarfTail_03}
      boxExtents=<20 30 30>
      constraintSetup {angularLimitsMax=<15 1 15>, angularLimitsMin=<-15 -45 -15>, angularTarget=<1 0 0>}
      localJointOffset=<4 0 0>
     }]
   }
   n23 = Anim Dynamics - Bone: dyn_scarfTail_04!(ComponentPose=@n22, Alpha=1, GravityOverride:Vector=@n18, ExternalForce="0,0,0") -> (Pose) {
    angularDampingOverride=1
    angularSpringConstant=20000
    bAngularSpring=true
    bOverrideAngularDamping=true
    bUseGravityOverride=true
    gravityScale=0.35
    simulationSpace=RootRelative
    alphaBoolBlend {blendOption=Linear}
    boundBone {boneName=dyn_scarfTail_04}
    componentLinearAccScale=<1 1 1>
    componentLinearVelScale=<1 1 1>
    gravityOverride {z=-980}
    physicsBodyDefinitions [{
      boundBone {boneName=dyn_scarfTail_04}
      boxExtents=<20 30 30>
      constraintSetup {angularLimitsMax=<15 0 15>, angularLimitsMin=<-15 -30 -15>, angularTarget=<1 0 0>}
      localJointOffset=<4 0 0>
     }]
   }
   n24 = "Transform (Modify) Bone - Bone: base_scarfTail"(ComponentPose=@n23, Alpha=1, Translation="-0.012,-0.000,0.531", Rotation="2.501,0.000,-0.000", Scale="1,1,1") {
    rotationMode=BMM_Additive
    rotationSpace=BCS_BoneSpace
    translationMode=BMM_Additive
    translationSpace=BCS_BoneSpace
    alphaBoolBlend {blendOption=Linear}
    boneToModify {boneName=base_scarfTail}
    rotation=<rot 2.501 0 0>
    translation=<-0.012 0 0.5312>
   }
   n25 = "Transform (Modify) Bone - Bone: dyn_scarfTail_01"(ComponentPose=@n24, Alpha=Get Is Crouching() -> (Is Crouching:float), Translation="0,0,0.3", Rotation="3,0,0", Scale="1,1,1") {
    rotationMode=BMM_Additive
    rotationSpace=BCS_BoneSpace
    translationMode=BMM_Additive
    translationSpace=BCS_BoneSpace
    alphaBoolBlend {blendOption=Linear}
    boneToModify {boneName=dyn_scarfTail_01}
    rotation {pitch=3}
    translation {z=0.3}
   }
   Save cached pose 'ScarfAnimToLegAnim'(Pose=Component To Local(ComponentPose=@n25)) {cacheName=ScarfAnimToLegAnim}
   n26 = Use cached pose 'conToScarfAnimDyn'()
   n27 = Use cached pose 'ScarfAnimToLegAnim'()
   n28 = Get Is Wearing Backbling()
   n29 = Anim Dynamics - Bone: dyn_pouchFlap!(
    ComponentPose=Local To Component(LocalPose=Blend Poses by bool(BlendPose_0=@n26, BlendPose_1=@n27, bActiveValue=@n28) {blendType=Linear})
    Alpha=1
    GravityOverride:Vector=Get GravityOverride()
    ExternalForce="0,0,0"
   ) -> (Pose) {
    angularDampingOverride=1
    angularSpringConstant=20000
    bAngularSpring=true
    bOverrideAngularDamping=true
    bUseGravityOverride=true
    gravityScale=0.35
    simulationSpace=RootRelative
    alphaBoolBlend {blendOption=Linear}
    boundBone {boneName=dyn_pouchFlap}
    componentLinearAccScale=<1 1 1>
    componentLinearVelScale=<1 1 1>
    gravityOverride {z=-980}
    physicsBodyDefinitions [{
      boundBone {boneName=dyn_pouchFlap}
      boxExtents=<20 30 30>
      constraintSetup {angularLimitsMax=<15 15 15>, angularLimitsMin=<-15 -15 -15>, angularTarget=<1 0 0>}
      localJointOffset=<4 0 0>
     }]
   }
   n30 = Get GravityOverride()
   n31 = Anim Dynamics - Bone: dyn_legZipper_l!(ComponentPose=@n29, Alpha=1, GravityOverride:Vector=@n30, ExternalForce="0,0,0") -> (Pose) {
    angularDampingOverride=1
    angularSpringConstant=20000
    bAngularSpring=true
    bOverrideAngularDamping=true
    bUseGravityOverride=true
    gravityScale=0.35
    simulationSpace=RootRelative
    alphaBoolBlend {blendOption=Linear}
    boundBone {boneName=dyn_legZipper_l}
    componentLinearAccScale=<1 1 1>
    componentLinearVelScale=<1 1 1>
    gravityOverride {z=-980}
    physicsBodyDefinitions [{
      boundBone {boneName=dyn_legZipper_l}
      boxExtents=<20 30 30>
      constraintSetup {angularLimitsMax=<15 15 15>, angularLimitsMin=<-15 -15 -15>, angularTarget=<1 0 0>}
      localJointOffset=<4 0 0>
     }]
   }
   n32 = Anim Dynamics - Bone: dyn_frontPant_l_02!(ComponentPose=@n31, Alpha=0.5, GravityOverride:Vector=@n30, ExternalForce="0,0,0") -> (Pose) {
    angularDampingOverride=1
    angularSpringConstant=20000
    bAngularSpring=true
    bOverrideAngularDamping=true
    bUseGravityOverride=true
    gravityScale=0.35
    simulationSpace=RootRelative
    alphaBoolBlend {blendOption=Linear}
    boundBone {boneName=dyn_frontPant_l_02}
    componentLinearAccScale=<1 1 1>
    componentLinearVelScale=<1 1 1>
    gravityOverride {z=-980}
    physicsBodyDefinitions [{
      boundBone {boneName=dyn_frontPant_l_02}
      boxExtents=<20 30 30>
      constraintSetup {angularLimitsMax=<15 15 15>, angularLimitsMin=<-15 -15 -15>, angularTarget=<1 0 0>}
      localJointOffset=<4 0 0>
     }]
   }
   n33 = Anim Dynamics - Bone: dyn_backPant_l_02!(ComponentPose=@n32, Alpha=0.5, GravityOverride:Vector=@n30, ExternalForce="0,0,0") -> (Pose) {
    angularDampingOverride=1
    angularSpringConstant=20000
    bAngularSpring=true
    bOverrideAngularDamping=true
    bUseGravityOverride=true
    gravityScale=0.35
    simulationSpace=RootRelative
    alphaBoolBlend {blendOption=Linear}
    boundBone {boneName=dyn_backPant_l_02}
    componentLinearAccScale=<1 1 1>
    componentLinearVelScale=<1 1 1>
    gravityOverride {z=-980}
    physicsBodyDefinitions [{
      boundBone {boneName=dyn_backPant_l_02}
      boxExtents=<20 30 30>
      constraintSetup {angularLimitsMax=<15 15 15>, angularLimitsMin=<-15 -15 -15>, angularTarget=<1 0 0>}
      localJointOffset=<4 0 0>
     }]
   }
   n34 = Get GravityOverride()
   n35 = Anim Dynamics - Bone: dyn_legZipper_r!(ComponentPose=@n33, Alpha=1, GravityOverride:Vector=@n34, ExternalForce="0,0,0") -> (Pose) {
    angularDampingOverride=1
    angularSpringConstant=20000
    bAngularSpring=true
    bOverrideAngularDamping=true
    bUseGravityOverride=true
    gravityScale=0.35
    simulationSpace=RootRelative
    alphaBoolBlend {blendOption=Linear}
    boundBone {boneName=dyn_legZipper_r}
    componentLinearAccScale=<1 1 1>
    componentLinearVelScale=<1 1 1>
    gravityOverride {z=-980}
    physicsBodyDefinitions [{
      boundBone {boneName=dyn_legZipper_r}
      boxExtents=<20 30 30>
      constraintSetup {angularLimitsMax=<15 15 15>, angularLimitsMin=<-15 -15 -15>, angularTarget=<1 0 0>}
      localJointOffset=<4 0 0>
     }]
   }
   n36 = Anim Dynamics - Bone: dyn_frontPant_r_02!(ComponentPose=@n35, Alpha=0.5, GravityOverride:Vector=@n34, ExternalForce="0,0,0") -> (Pose) {
    angularDampingOverride=1
    angularSpringConstant=20000
    bAngularSpring=true
    bOverrideAngularDamping=true
    bUseGravityOverride=true
    gravityScale=0.35
    simulationSpace=RootRelative
    alphaBoolBlend {blendOption=Linear}
    boundBone {boneName=dyn_frontPant_r_02}
    componentLinearAccScale=<1 1 1>
    componentLinearVelScale=<1 1 1>
    gravityOverride {z=-980}
    physicsBodyDefinitions [{
      boundBone {boneName=dyn_frontPant_r_02}
      boxExtents=<20 30 30>
      constraintSetup {angularLimitsMax=<15 15 15>, angularLimitsMin=<-15 -15 -15>, angularTarget=<1 0 0>}
      localJointOffset=<4 0 0>
     }]
   }
   n37 = Anim Dynamics - Bone: dyn_backPant_r_02!(ComponentPose=@n36, Alpha=0.5, GravityOverride:Vector=@n34, ExternalForce="0,0,0") -> (Pose) {
    angularDampingOverride=1
    angularSpringConstant=20000
    bAngularSpring=true
    bOverrideAngularDamping=true
    bUseGravityOverride=true
    gravityScale=0.35
    simulationSpace=RootRelative
    alphaBoolBlend {blendOption=Linear}
    boundBone {boneName=dyn_backPant_r_02}
    componentLinearAccScale=<1 1 1>
    componentLinearVelScale=<1 1 1>
    gravityOverride {z=-980}
    physicsBodyDefinitions [{
      boundBone {boneName=dyn_backPant_r_02}
      boxExtents=<20 30 30>
      constraintSetup {angularLimitsMax=<15 15 15>, angularLimitsMin=<-15 -15 -15>, angularTarget=<1 0 0>}
      localJointOffset=<4 0 0>
     }]
   }
   Save cached pose 'AnimToLogic'(Pose=Component To Local(ComponentPose=@n37)) {cacheName=AnimToLogic}
   n38 = Get Is Crouching()
   n39 = "Transform (Modify) Bone - Bone: dyn_frontPant_l_01"(
    ComponentPose=Local To Component(LocalPose=Use cached pose 'AnimToLogic'())
    Alpha=@n38
    Translation="0,0,0"
    Rotation="16.5,4,0"
    Scale="1,1,1"
   ) {
    rotationMode=BMM_Additive
    rotationSpace=BCS_BoneSpace
    translationSpace=BCS_BoneSpace
    alphaBoolBlend {blendOption=Linear}
    boneToModify {boneName=dyn_frontPant_l_01}
    rotation {pitch=16.5, yaw=4}
   }
   n40 = "Transform (Modify) Bone - Bone: dyn_frontPant_l_02"(ComponentPose=@n39, Alpha=@n38, Translation="0,0,0", Rotation="10,10,0", Scale="1,1,1") {
    rotationMode=BMM_Additive
    rotationSpace=BCS_BoneSpace
    translationSpace=BCS_BoneSpace
    alphaBoolBlend {blendOption=Linear}
    boneToModify {boneName=dyn_frontPant_l_02}
    rotation {pitch=10, yaw=10}
   }
   n41 = Get Is Crouching()
   n42 = "Transform (Modify) Bone - Bone: dyn_frontPant_r_01"(ComponentPose=@n40, Alpha=@n41, Translation="-1,0,0", Rotation="12.5,8,5", Scale="1,1,1") {
    rotationMode=BMM_Additive
    rotationSpace=BCS_BoneSpace
    translationMode=BMM_Additive
    translationSpace=BCS_BoneSpace
    alphaBoolBlend {blendOption=Linear}
    boneToModify {boneName=dyn_frontPant_r_01}
    rotation=<rot 12.5 8 5>
    translation {x=-1}
   }
   n43 = "Transform (Modify) Bone - Bone: dyn_frontPant_r_02"(ComponentPose=@n42, Alpha=@n41, Translation="-1,0,0", Rotation="14,4,5", Scale="1,1,1") {
    rotationMode=BMM_Additive
    rotationSpace=BCS_BoneSpace
    translationMode=BMM_Additive
    translationSpace=BCS_BoneSpace
    alphaBoolBlend {blendOption=Linear}
    boneToModify {boneName=dyn_frontPant_r_02}
    rotation=<rot 14 4 5>
    translation {x=-1}
   }
   n44 = "Transform (Modify) Bone - Bone: dyn_backPant_r_02"(ComponentPose=@n43, Alpha=@n41, Translation="0,0,0", Rotation="-7,0,-7", Scale="1,1,1") {rotationMode=BMM_Additive, alphaBoolBlend {blendOption=Linear}, boneToModify {boneName=dyn_backPant_r_02}, rotation {pitch=-7, roll=-7}}
   n45 = "Transform (Modify) Bone - Bone: base_chainTop"!(ComponentPose=@n44, Alpha:float=Get Is Crouching(), Translation="0,0,0", Rotation="10,-10,0", Scale="1,1,1") -> (Pose) {
    rotationMode=BMM_Additive
    rotationSpace=BCS_BoneSpace
    translationMode=BMM_Additive
    translationSpace=BCS_BoneSpace
    alphaBoolBlend {blendOption=Linear}
    boneToModify {boneName=base_chainTop}
    rotation {pitch=10, yaw=-10}
   }
   n46 = "Transform (Modify) Bone - Bone: dyn_heart_l"(ComponentPose=@n45, Alpha=Get Is Crouching(), Translation="0.5,0,0", Rotation="0,0,0", Scale="1,1,1") {
    rotationSpace=BCS_BoneSpace
    translationMode=BMM_Additive
    translationSpace=BCS_BoneSpace
    alphaBoolBlend {blendOption=Linear}
    boneToModify {boneName=dyn_heart_l}
    translation {x=0.5}
   }
   Save cached pose 'CLogToSDLog'(Pose=Component To Local(ComponentPose=@n46)) {cacheName=CLogToSDLog}
  ]}
 EventGraph {nodes [
   Event BlueprintUpdateAnimation()
   > n1 = Sequence()
   > n2 = Gravity Override State Function DEPRECATED(
    CharacterPartAnimInstance=Self-Reference()
    GameTimeInSeconds=GetGameTimeInSeconds()
    GravityOverrideParams=Get GravityOverrideParams() -> (GravityOverrideParams:Gravity Override Params Struct Structure)
   )
   > Set GravityOverride(GravityOverride=@n2)
   > n3 = Branch(Condition=Get bIsCrouching())
   > Set Is Crouching(Is Crouching:float=1) -> (Output_Get:float)
   Set Is Crouching(execute=@n3.else, Is Crouching=0)
   > n4 = Branch(Condition=Get bIsSkyDiving())
   > Set Is Sky Diving(Is Sky Diving:double=1) -> (Output_Get:double)
   Set Is Sky Diving(execute=@n4.else, Is Sky Diving=0)
   n5 = GetPartSkeletalMeshComponent()
   n6 = FindChildCommponentFromAttachSocket!(
    self:Custom Character Part Anim Instance Object Reference
    Parent:Skeletal Mesh Component Object Reference=@n5
    bIncludeAllDescendants:bool=false
    AttachSocket:Name=Cape
   ) -> (ReturnValue:Skeletal Mesh Component Object Reference)
   n7 = FindChildCommponentFromAttachSocket(Parent=@n5, bIncludeAllDescendants=false, AttachSocket=backpack_br)
   Set Is Wearing Backbling(
    execute=@n1.then_1
    Is Wearing Backbling=OR Boolean(
     A="Not Equal (Object)"(A:Object Reference=@n6, B:Object Reference) -> (ReturnValue:bool=~) {is=title K2Node_PromotableOperator}
     B="Not Equal (Object)"(A=@n7) -> (ReturnValue=~)
    )
   )
  ]}
]
variables [
 GravityOverride {type=Vector}
 GravityOverrideParams {type=GravityOverrideParamsStruct}
 Is Crouching {type=number}
 Is Sky Diving {type=number}
 Is Wearing Backbling {type=boolean}
]
