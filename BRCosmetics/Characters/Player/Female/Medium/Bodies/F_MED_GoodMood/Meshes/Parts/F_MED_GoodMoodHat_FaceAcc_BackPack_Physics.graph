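# Physics asset for the hat, face accessory and backpack part of the medium female GoodMood cosmetic character, defining capsule bodies mainly for hair dynamics.
# Back and front hair chains (center, left, right) hang from base bodies on the head.
# Back chains lock twist, with swing limits generally widening toward the tips, while front chains and base joints are unconstrained in rotation.
# Head, neck, clavicle and multi-capsule upper-spine bodies form the body volumes around the hair.
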PhysicsAsset /BRCosmetics/Characters/Player/Female/Medium/Bodies/F_MED_GoodMood/Meshes/Parts/F_MED_GoodMoodHat_FaceAcc_BackPack_Physics
bodies {
 Base_Hair_Back_c [{shapeType=Capsule, length=0.505, radius=0.505, shapeName=Base_Hair_Back_c_capsule, rotation=<rot 0 -90 0>}]
 Base_Hair_Back_l [{shapeType=Capsule, length=0.505, radius=0.505, shapeName=Base_Hair_Back_l_capsule, rotation=<rot 0 -90 0>}]
 Base_Hair_Back_r [{shapeType=Capsule, length=0.505, radius=0.505, shapeName=Base_Hair_Back_r_capsule, rotation=<rot 0 -90 0>}]
 Base_Hair_Front_c [{shapeType=Capsule, length=0.505, radius=0.505, shapeName=Base_Hair_Front_c_capsule, rotation=<rot 0 -90 0>}]
 Base_Hair_Front_l [{shapeType=Capsule, length=0.505, radius=0.505, shapeName=Base_Hair_Front_l_capsule, rotation=<rot 0 -90 0>}]
 Base_Hair_Front_r [{shapeType=Capsule, length=0.505, radius=0.505, shapeName=Base_Hair_Front_r_capsule, rotation=<rot 0 -90 0>}]
 Dyn_Hair_Back_01_c [{shapeType=Capsule, length=2.5, radius=2.5, shapeName=Dyn_Hair_Back_01_c_capsule, center=<1 0 0>, rotation=<rot 90 0 0>}]
 Dyn_Hair_Back_01_l [{shapeType=Capsule, length=2.5, radius=2.5, shapeName=Dyn_Hair_Back_01_l_capsule, center=<1 0 0>, rotation=<rot 90 0 0>}]
 Dyn_Hair_Back_01_r [{shapeType=Capsule, length=2.5, radius=2.5, shapeName=Dyn_Hair_Back_01_r_capsule, center=<1 0 0>, rotation=<rot 90 0 0>}]
 Dyn_Hair_Back_02_c [{shapeType=Capsule, length=2.7948, radius=2.2052, shapeName=Dyn_Hair_Back_02_c_capsule, center=<1 0 0>, rotation=<rot 90 0 0>}]
 Dyn_Hair_Back_02_l [{shapeType=Capsule, length=2.7948, radius=2.2052, shapeName=Dyn_Hair_Back_02_l_capsule, center=<1 0 0>, rotation=<rot 90 0 0>}]
 Dyn_Hair_Back_02_r [{
   shapeType=Capsule
   length=4.7875
   radius=1.7125
   shapeName=dyn_hair_back_02_r_capsule
   center=<5.1075 -0.6879 -1.5848>
   rotation=<rot -8.4373 -12.4483 76.3125>
  }]
 Dyn_Hair_Back_03_c [{shapeType=Capsule, length=1.7157, radius=2, shapeName=Dyn_Hair_Back_03_c_capsule, center=<1 0 -0.3666>, rotation=<rot 90 0 0>}]
 Dyn_Hair_Back_03_l [{shapeType=Capsule, length=3.7157, radius=1.5, shapeName=Dyn_Hair_Back_03_l_capsule, center=<1 0 0>, rotation=<rot 90 0 0>}]
 Dyn_Hair_Back_03_r [{shapeType=Capsule, length=5, radius=2, shapeName=dyn_hair_back_03_r_capsule, center=<4.0401 -1.8408 -0.3335>, rotation=<rot 90 0 0>}]
 Dyn_Hair_Back_04_c [{shapeType=Capsule, length=2.3173, radius=2, shapeName=Dyn_Hair_Back_04_c_capsule, center=<1.2611 0 0.0256>, rotation=<rot 75.0003 -180 -180>}]
 Dyn_Hair_Back_04_l [{shapeType=Capsule, length=4, radius=2, shapeName=Dyn_Hair_Back_03_l_capsule, center=<3.2768 0 1.4569>, rotation=<rot 80.0001 -180 -180>}]
 Dyn_Hair_Back_04_r [{shapeType=Capsule, length=4.5, radius=1.5, shapeName=dyn_hair_back_04_r_capsule, center=<4.4242 0 0.7018>, rotation=<rot 90 0 0>}]
 Dyn_Hair_Back_05_c [{shapeType=Capsule, length=2.5, radius=2.5, shapeName=Dyn_Hair_Back_05_c_capsule, center=<1 0 0>, rotation=<rot 90 0 0>}]
 Dyn_Hair_Front_01_c [{
   shapeType=Capsule
   length=3.8234
   radius=2.6872
   shapeName=Dyn_Hair_Front_01_c_capsule
   center=<3.9614 -0.8036 0.0101>
   rotation=<rot 0 51.5314 -87.5766>
  }]
 Dyn_Hair_Front_01_l [{
   shapeType=Capsule
   length=3.0731
   radius=2.6238
   shapeName=Dyn_Hair_Front_01_l_capsule
   center=<2.1758 -1.0683 -0.3565>
   rotation=<rot 0 178.5362 -78.0255>
  }]
 Dyn_Hair_Front_01_r [{
   shapeType=Capsule
   length=4.2477
   radius=3.1302
   shapeName=Dyn_Hair_Front_01_r_capsule
   center=<3.1662 -0.0774 0.0016>
   rotation=<rot 0 10.4918 -74.7145>
  }]
 Dyn_Hair_Front_02_r [{
   shapeType=Capsule
   length=3.3866
   radius=3.3224
   shapeName=Dyn_Hair_Front_02_r_capsule
   center=<2.5906 -0.0475 -0.6683>
   rotation=<rot 36.5522 -109.7947 -90>
  }]
 Dyn_Hair_Front_03_r [{
   shapeType=Capsule
   length=3.2836
   radius=2.1647
   shapeName=Dyn_Hair_Front_03_r_capsule
   center=<2.2915 0.8931 -0.7202>
   rotation=<rot 0 -78.0424 -83.3442>
  }]
 clavicle_l [{shapeType=Capsule, length=5.8315, radius=9.906, shapeName=clavicle_l_capsule, center=<3.887 -0.3924 -1.8983>, rotation=<rot -0.0001 -90 90>}]
 clavicle_r [{shapeType=Capsule, length=3.9665, radius=8.1199, shapeName=clavicle_l_capsule, center=<-3.887 0 1.45>, rotation=<rot -0.0001 -90 -90>}]
 head [{shapeType=Capsule, length=3.969, radius=5.5623, center=<1.8699 0 0>, rotation=<rot 90 180 -180>}]
 neck_01 [{shapeType=Capsule, length=4.7091, radius=4.183, shapeName=neck_01_capsule, center=<3.5215 0 0>, rotation=<rot 0 -90.0003 90>}]
 spine_05 [
  {shapeType=Capsule, length=37.6833, radius=17.6865, shapeName=spine_05_capsule, center=<4.1042 -33.8984 -0.0044>, rotation=<rot 0 -90 0.0616>}
  {shapeType=Capsule, length=53.5569, radius=15, shapeName=spine_05_capsule, center=<12.5089 19.199 0>}
  {shapeType=Capsule, length=16.326, radius=13.674, shapeName=spine_05_capsule}
  {shapeType=Capsule, length=40.5375, radius=15, shapeName=spine_05_capsule, center=<24.5612 -23.7639 0>}
 ]
}
constraints [
 {
  bone1Name=Base_Hair_Back_c
  bone2Name=head
  swing1LimitDegrees=45
  swing1Motion=Free
  swing2LimitDegrees=45
  swing2Motion=Free
  twistLimitDegrees=45
  twistMotion=Free
 }
 {
  bone1Name=Dyn_Hair_Back_01_c
  bone2Name=Base_Hair_Back_c
  swing1LimitDegrees=1
  swing1Motion=Locked
  swing2LimitDegrees=10
  twistLimitDegrees=0
  twistMotion=Locked
 }
 {
  bone1Name=Dyn_Hair_Back_02_c
  bone2Name=Dyn_Hair_Back_01_c
  swing1LimitDegrees=10
  swing2LimitDegrees=15
  twistLimitDegrees=45
  twistMotion=Locked
 }
 {
  bone1Name=Dyn_Hair_Back_03_c
  bone2Name=Dyn_Hair_Back_02_c
  swing1LimitDegrees=10
  swing2LimitDegrees=30
  twistLimitDegrees=45
  twistMotion=Locked
 }
 {
  bone1Name=Dyn_Hair_Back_04_c
  bone2Name=Dyn_Hair_Back_03_c
  swing1LimitDegrees=20
  swing2LimitDegrees=40
  twistLimitDegrees=45
  twistMotion=Locked
 }
 {
  bone1Name=Dyn_Hair_Back_05_c
  bone2Name=Dyn_Hair_Back_04_c
  swing1LimitDegrees=30
  swing2LimitDegrees=30
  twistLimitDegrees=45
  twistMotion=Locked
 }
 {
  bone1Name=Base_Hair_Back_l
  bone2Name=head
  swing1LimitDegrees=45
  swing1Motion=Free
  swing2LimitDegrees=45
  swing2Motion=Free
  twistLimitDegrees=45
  twistMotion=Free
 }
 {
  bone1Name=Dyn_Hair_Back_02_l
  bone2Name=Dyn_Hair_Back_01_l
  swing1LimitDegrees=5
  swing2LimitDegrees=15
  twistLimitDegrees=45
  twistMotion=Locked
 }
 {
  bone1Name=Dyn_Hair_Back_03_l
  bone2Name=Dyn_Hair_Back_02_l
  swing1LimitDegrees=10
  swing2LimitDegrees=30
  twistLimitDegrees=45
  twistMotion=Locked
 }
 {
  bone1Name=Dyn_Hair_Back_01_l
  bone2Name=Base_Hair_Back_l
  swing1LimitDegrees=1
  swing1Motion=Locked
  swing2LimitDegrees=10
  twistLimitDegrees=0
  twistMotion=Locked
 }
 {
  bone1Name=Base_Hair_Back_r
  bone2Name=head
  swing1LimitDegrees=45
  swing1Motion=Free
  swing2LimitDegrees=45
  swing2Motion=Free
  twistLimitDegrees=45
  twistMotion=Free
 }
 {
  bone1Name=Dyn_Hair_Back_01_r
  bone2Name=Base_Hair_Back_r
  swing1LimitDegrees=1
  swing1Motion=Locked
  swing2LimitDegrees=10
  twistLimitDegrees=0
  twistMotion=Locked
 }
 {
  bone1Name=Dyn_Hair_Back_02_r
  bone2Name=Dyn_Hair_Back_01_r
  swing1LimitDegrees=5
  swing2LimitDegrees=15
  twistLimitDegrees=45
  twistMotion=Locked
 }
 {
  bone1Name=Dyn_Hair_Back_03_r
  bone2Name=Dyn_Hair_Back_02_r
  swing1LimitDegrees=10
  swing2LimitDegrees=30
  twistLimitDegrees=45
  twistMotion=Locked
 }
 {
  bone1Name=Dyn_Hair_Back_04_r
  bone2Name=Dyn_Hair_Back_03_r
  swing1LimitDegrees=20
  swing2LimitDegrees=40
  twistLimitDegrees=45
  twistMotion=Locked
 }
 {
  bone1Name=head
  bone2Name=neck_01
  swing1LimitDegrees=45
  swing1Motion=Free
  swing2LimitDegrees=45
  swing2Motion=Free
  twistLimitDegrees=45
  twistMotion=Free
 }
 {
  bone1Name=clavicle_l
  bone2Name=spine_05
  swing1LimitDegrees=45
  swing1Motion=Free
  swing2LimitDegrees=45
  swing2Motion=Free
  twistLimitDegrees=45
  twistMotion=Free
 }
 {
  bone1Name=neck_01
  bone2Name=spine_05
  swing1LimitDegrees=45
  swing1Motion=Free
  swing2LimitDegrees=45
  swing2Motion=Free
  twistLimitDegrees=45
  twistMotion=Free
 }
 {
  bone1Name=clavicle_r
  bone2Name=spine_05
  swing1LimitDegrees=45
  swing1Motion=Free
  swing2LimitDegrees=45
  swing2Motion=Free
  twistLimitDegrees=45
  twistMotion=Free
 }
 {
  bone1Name=Dyn_Hair_Front_01_r
  bone2Name=Base_Hair_Front_r
  swing1LimitDegrees=45
  swing1Motion=Free
  swing2LimitDegrees=45
  swing2Motion=Free
  twistLimitDegrees=45
  twistMotion=Free
 }
 {
  bone1Name=Dyn_Hair_Front_02_r
  bone2Name=Dyn_Hair_Front_01_r
  swing1LimitDegrees=45
  swing1Motion=Free
  swing2LimitDegrees=45
  swing2Motion=Free
  twistLimitDegrees=45
  twistMotion=Free
 }
 {
  bone1Name=Dyn_Hair_Front_03_r
  bone2Name=Dyn_Hair_Front_02_r
  swing1LimitDegrees=45
  swing1Motion=Free
  swing2LimitDegrees=45
  swing2Motion=Free
  twistLimitDegrees=45
  twistMotion=Free
 }
 {
  bone1Name=Base_Hair_Front_r
  bone2Name=head
  swing1LimitDegrees=45
  swing1Motion=Free
  swing2LimitDegrees=45
  swing2Motion=Free
  twistLimitDegrees=45
  twistMotion=Free
 }
 {
  bone1Name=Base_Hair_Front_c
  bone2Name=head
  swing1LimitDegrees=45
  swing1Motion=Free
  swing2LimitDegrees=45
  swing2Motion=Free
  twistLimitDegrees=45
  twistMotion=Free
 }
 {
  bone1Name=Dyn_Hair_Front_01_c
  bone2Name=Base_Hair_Front_c
  swing1LimitDegrees=45
  swing1Motion=Free
  swing2LimitDegrees=45
  swing2Motion=Free
  twistLimitDegrees=45
  twistMotion=Free
 }
 {
  bone1Name=Base_Hair_Front_l
  bone2Name=head
  swing1LimitDegrees=45
  swing1Motion=Free
  swing2LimitDegrees=45
  swing2Motion=Free
  twistLimitDegrees=45
  twistMotion=Free
 }
 {
  bone1Name=Dyn_Hair_Front_01_l
  bone2Name=Base_Hair_Front_l
  swing1LimitDegrees=45
  swing1Motion=Free
  swing2LimitDegrees=45
  swing2Motion=Free
  twistLimitDegrees=45
  twistMotion=Free
 }
 {
  bone1Name=Dyn_Hair_Back_04_l
  bone2Name=Dyn_Hair_Back_03_l
  swing1LimitDegrees=20
  swing2LimitDegrees=45
  twistLimitDegrees=45
  twistMotion=Locked
 }
]
