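# Control Rig for the female medium GoodMood hat face accessory that adjusts back hair and hat bones as the head moves.
# A spherical pose reader on the head bone, with debug drawing left on, feeds two cubic remap curves with different input thresholds.
# The curves weight two Modify Transforms steps that offset Base_Hair_Back and Base_Hat_Back bones, carrying their child Dyn_ bones along.
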ControlRigBlueprint /BRCosmetics/Characters/Player/Female/Medium/Bodies/F_MED_GoodMood/Meshes/Parts/F_MED_GoodMoodHat_FaceAcc_CtrlRig
graphs [
 RigVMFunctionLibrary {links [], nodes []}
 RigVMModel {nodes [
   RigUnit_BeginExecution = &RigUnit_BeginExecution()
   RigVMFunction_Sequence = &RigVMFunction_Sequence(ExecuteContext=@RigUnit_BeginExecution.ExecutePin)
   n1 = &RigUnit_SphericalPoseReader(
    ActiveRegionScaleFactors=(PositiveWidth=1,NegativeWidth=1,PositiveHeight=1.5,NegativeHeight=1.5)
    ActiveRegionSize=0.1
    Debug=(bDrawDebug=True,bDraw2D=false,bDrawLocalAxes=true,DebugScale=25,DebugSegments=20,DebugThickness=0.25)
    DriverAxis=(X=0,Y=1,Z=0)
    DriverItem=(Type=Bone,Name="head")
    FalloffRegionScaleFactors=(PositiveWidth=1,NegativeWidth=1.5,PositiveHeight=2.5,NegativeHeight=2.5)
    FalloffSize=0.2
    FlipHeightScaling=False
    FlipWidthScaling=False
    OptionalParentItem=(Type=Bone,Name="None")
    RotationOffset=(X=90,Y=0,Z=-45)
    ExecutePin=@RigVMFunction_Sequence.A
   )
   RigVMFunction_AnimEvalRichCurve = &RigVMFunction_AnimEvalRichCurve(
    Curve=(EditorCurveData=(Keys=((InterpMode=RCIM_Cubic,ArriveTangent=2.3755),(InterpMode=RCIM_Cubic,Time=1,Value=1,LeaveTangent=10.177))))
    SourceMinimum=0.2
    TargetMinimum=0
    Value=@n1.OutputParam
   )
   > &RigUnit_ModifyTransforms(
    ItemToModify="((Item=(Type=Bone,Name=\"Base_Hair_Back_c\"),Transform=(Rotation=(X=0,Y=-0.3827,Z=0,W=0.9239),Translation=(X=0,Y=0,Z=0),Scale3D=(X=1,Y=1,Z=1))),(Item=(Type=Bone,Name=\"Base_Hair_Back_l\"),Transform=(Rotation=(X=0,Y=-0.3827,Z=0,W=0.9239),Translation=(X=0,Y=0,Z= ...(truncated)"
    WeightMinimum=0
    Weight=@RigVMFunction_AnimEvalRichCurve.Result
   )
   n2 = &RigVMFunction_AnimEvalRichCurve(
    Curve=(EditorCurveData=(Keys=((InterpMode=RCIM_Cubic,ArriveTangent=1.4219),(InterpMode=RCIM_Cubic,Time=1,Value=1,LeaveTangent=4836.3604))))
    SourceMinimum=0.3
    TargetMinimum=0
    Value=@n1.OutputParam
   )
   > &RigUnit_ModifyTransforms(
    ItemToModify="((Item=(Type=Bone,Name=\"Base_Hat_Back_l\"),Transform=(Rotation=(X=0.1245,Y=-0.2981,Z=0.0393,W=0.9456),Translation=(X=-2,Y=0,Z=0),Scale3D=(X=1,Y=1,Z=1))),(Item=(Type=Bone,Name=\"Base_Hat_Back_c\"),Transform=(Rotation=(X=-0.2296,Y=-0.446,Z=-0.1195,W=0.8568),Tra ...(truncated)"
    WeightMinimum=0
    Weight=@n2.Result
   )
  ]}
]
hierarchy [
 root {loc=<0 0 0>, rot=<0 0 0 1>}
 pelvis {parent=root, loc=<0 -0.5301 98.0571>, rot=<0.0127 0.707 -0.0127 -0.707>}
 spine_01 {parent=pelvis, loc=<6.1208 -0.9328 0>, rot=<0 0 0.034 -0.9994>}
 spine_02 {parent=spine_01, loc=<6.5996 -0.2095 0>, rot=<0 0 0.0238 -0.9997>}
 spine_03 {parent=spine_02, loc=<6.4914 -0.1237 0>, rot=<0 0 0.0149 -0.9999>}
 spine_04 {parent=spine_03, loc=<6.9598 0.0794 0>, rot=<0 0 -0.0132 -0.9999>}
 spine_05 {parent=spine_04, loc=<6.9987 0 0>, rot=<0 0 -0.0015 -1>}
 neck_01 {parent=spine_05, loc=<11.3138 0.1529 0>, rot=<0 0 -0.2003 -0.9797>}
 neck_02 {parent=neck_01, loc=<4.5627 0 0>, rot=<0 0 0.0504 -0.9987>}
 head {parent=neck_02, loc=<5.3914 0 0>, rot=<0 0 0.115 -0.9934>}
 Base_Hat_Back_c {parent=head, loc=<6.8614 -1.485 0.2915>, rot=<0.5 -0.5 0.5 -0.5>}
 Dyn_Hat_Back_01_c {parent=Base_Hat_Back_c, loc=<0 0 0>, rot=<0 0 0 1>}
 Base_Hat_Front_c {parent=head, loc=<8.116 1.1036 0.2915>, rot=<-0.5 -0.5 -0.5 -0.5>}
 Dyn_Hat_Front_01_c {parent=Base_Hat_Front_c, loc=<0 0 0>, rot=<0 0 0 1>}
 Base_Hat_r {parent=head, loc=<7.8028 -0.1142 -0.5934>, rot=<0 -0.8195 0 -0.5731>}
 Dyn_Hat_01_r {parent=Base_Hat_r, loc=<0 0 0>, rot=<0 0 0 1>}
 Base_Hat_l {parent=head, loc=<7.3974 -0.1176 1.9988>, rot=<0.7071 0 0.7071 0>}
 Dyn_Hat_01_l {parent=Base_Hat_l, loc=<0 0 0>, rot=<0 0 0 1>}
 Base_Fether_01 {parent=head, loc=<12.2543 1.4299 -6.0892>, rot=<0.6723 -0.1822 0.1759 -0.6956>}
 Dyn_Fether_01 {parent=Base_Fether_01, loc=<0 0 0>, rot=<0 0 0 1>}
 Base_Fether_02 {parent=head, loc=<10.2208 -1.1169 -6.4173>, rot=<0.528 -0.3805 0.2836 -0.7043>}
 Dyn_Fether_02 {parent=Base_Fether_02, loc=<0 0 0>, rot=<0 0 0 1>}
 Base_Hair_Front_l {parent=head, loc=<7.7779 7.3389 -4.7918>, rot=<0.3668 -0.8934 -0.2308 -0.1182>}
 Dyn_Hair_Front_01_l {parent=Base_Hair_Front_l, loc=<0 0 0>, rot=<0 0 0 1>}
 Base_Hair_Front_c {parent=head, loc=<8.671 9.6608 -0.3736>, rot=<0.123 -0.8148 -0.5268 -0.2087>}
 Dyn_Hair_Front_01_c {parent=Base_Hair_Front_c, loc=<0 0 0>, rot=<0 0 0 1>}
 Base_Hair_Front_r {parent=head, loc=<9.1462 4.3707 5.7701>, rot=<0.1967 0.2282 0.9533 -0.0218>}
 Dyn_Hair_Front_01_r {parent=Base_Hair_Front_r, loc=<0 0 0>, rot=<0 0 0 1>}
 Dyn_Hair_Front_02_r {parent=Dyn_Hair_Front_01_r, loc=<6.0704 0 0>, rot=<0.0109 -0.2666 0.0122 -0.9637>}
 Dyn_Hair_Front_03_r {parent=Dyn_Hair_Front_02_r, loc=<6.3167 0 0>, rot=<0 -0.1227 0 -0.9924>}
 Base_Earing_l {parent=head, loc=<-1.761 -0.0002 -6.0243>, rot=<0 0.7071 0.7071 0>}
 Dyn_Earing_01_l {parent=Base_Earing_l, loc=<0 0 0>, rot=<0 0 0 1>}
 Dyn_Earing_02_l {parent=Dyn_Earing_01_l, loc=<1.4139 0 0>, rot=<0 0 0 1>}
 Base_Earing_r {parent=head, loc=<-1.6998 -0.0004 5.9962>, rot=<0 0.7071 0.7071 0>}
 Dyn_Earing_01_r {parent=Base_Earing_r, loc=<0 0 0>, rot=<0 0 0 1>}
 Dyn_Earing_02_r {parent=Dyn_Earing_01_r, loc=<1.4139 0 0>, rot=<0 0 0 1>}
 Base_Hat_Back_l {parent=head, loc=<7.3974 -1.0846 1.3291>, rot=<0.6708 -0.2237 0.6708 -0.2237>}
 Dyn_Hat_Back_l {parent=Base_Hat_Back_l, loc=<0 0 0>, rot=<0 0 0 1>}
 head_fix {parent=head, loc=<0 -6 0>, rot=<0.0029 -0.7071 -0.0029 -0.7071>}
 Base_Hair_Back_l {parent=head, loc=<8.8514 -4.3917 -4.697>, rot=<0.0029 -0.7071 0.7071 -0.0029>}
 Dyn_Hair_Back_01_l {parent=Base_Hair_Back_l, loc=<0 0 0>, rot=<0 0 0 1>}
 Dyn_Hair_Back_02_l {parent=Dyn_Hair_Back_01_l, loc=<7.848 0 0>, rot=<0 0.191 0 -0.9816>}
 Dyn_Hair_Back_03_l {parent=Dyn_Hair_Back_02_l, loc=<5.895 0 0>, rot=<0 0.191 0 -0.9816>}
 Dyn_Hair_Back_04_l {parent=Dyn_Hair_Back_03_l, loc=<5.895 0 0>, rot=<0 -0.2177 0 -0.976>}
 Dyn_Hair_Back_05_l {parent=Dyn_Hair_Back_04_l, loc=<6.0501 0 0>, rot=<0 -0.1558 0 -0.9878>}
 Base_Hair_Back_c {parent=head, loc=<8.8688 -6.5046 0>, rot=<0.0029 -0.7071 0.7071 -0.0029>}
 Dyn_Hair_Back_01_c {parent=Base_Hair_Back_c, loc=<0 0 0>, rot=<0 0 0 1>}
 Dyn_Hair_Back_02_c {parent=Dyn_Hair_Back_01_c, loc=<7.848 0 0.706>, rot=<0 0.191 0 -0.9816>}
 Dyn_Hair_Back_03_c {parent=Dyn_Hair_Back_02_c, loc=<5.895 0 0>, rot=<0 0.0754 0 -0.9972>}
 Dyn_Hair_Back_04_c {parent=Dyn_Hair_Back_03_c, loc=<5.895 0 0>, rot=<0 -0.175 0 -0.9846>}
 Dyn_Hair_Back_05_c {parent=Dyn_Hair_Back_04_c, loc=<5.582 0 0>, rot=<0 -0.1942 0 -0.981>}
 Base_Hair_Back_r {parent=head, loc=<8.8514 -4.3917 4.697>, rot=<0.0029 -0.7071 0.7071 -0.0029>}
 Dyn_Hair_Back_01_r {parent=Base_Hair_Back_r, loc=<0 0 0>, rot=<0 0 0 1>}
 Dyn_Hair_Back_02_r {parent=Dyn_Hair_Back_01_r, loc=<7.848 0 0>, rot=<0 0.191 0 -0.9816>}
 Dyn_Hair_Back_03_r {parent=Dyn_Hair_Back_02_r, loc=<5.895 0 0>, rot=<0 0.191 0 -0.9816>}
 Dyn_Hair_Back_04_r {parent=Dyn_Hair_Back_03_r, loc=<5.895 0 0>, rot=<0 -0.175 0 -0.9846>}
 Dyn_Hair_Back_05_r {parent=Dyn_Hair_Back_04_r, loc=<6.0501 0 0>, rot=<0 -0.1558 0 -0.9878>}
 clavicle_l {parent=spine_05, loc=<5.4831 -0.1053 -3.2946>, rot=<0.0465 -0.7289 -0.009 -0.6829>}
 upperarm_l {parent=clavicle_l, loc=<9.8658 -1.2136 -0.4105>, rot=<0.0164 -0.3245 -0.0269 -0.9454>}
 upperarm_twist_01_l {parent=upperarm_l, loc=<7.2135 -0.1134 0>, rot=<0 0 0 1>}
 clavicle_r {parent=spine_05, loc=<5.4831 -0.1053 3.2946>, rot=<0.7289 0.0465 -0.6829 0.009>}
 upperarm_r {parent=clavicle_r, loc=<-9.8658 1.2136 0.4105>, rot=<0.0164 -0.3245 -0.0269 -0.9454>}
 upperarm_twist_01_r {parent=upperarm_r, loc=<-7.2135 0.1134 0>, rot=<0 0 0 -1>}
]
