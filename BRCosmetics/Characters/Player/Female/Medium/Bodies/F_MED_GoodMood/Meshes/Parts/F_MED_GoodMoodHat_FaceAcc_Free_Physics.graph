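# Physics asset for the female medium GoodMood hat face accessory, simulating its dynamic hair on the player character.
# Four capsule chains (back center, left, right and front right) hang from small base bodies linked to the head, alongside head, neck, upper spine and clavicle capsules.
# Chain joints limit or lock twist and mostly widen swing toward the tips.
# An extra all-free angular constraint links the right back tip to the center chain.
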PhysicsAsset /BRCosmetics/Characters/Player/Female/Medium/Bodies/F_MED_GoodMood/Meshes/Parts/F_MED_GoodMoodHat_FaceAcc_Free_Physics
bodies {
 Base_Hair_Back_c [{shapeType=Capsule, length=0.505, radius=0.505, shapeName=Base_Hair_Back_c_capsule, rotation=<rot 0 -90 0>}]
 Base_Hair_Back_l [{shapeType=Capsule, length=0.505, radius=0.505, shapeName=Base_Hair_Back_l_capsule, rotation=<rot 0 -90 0>}]
 Base_Hair_Back_r [{shapeType=Capsule, length=0.505, radius=0.505, shapeName=Base_Hair_Back_r_capsule, rotation=<rot 0 -90 0>}]
 Base_Hair_Front_r [{shapeType=Capsule, length=0.505, radius=0.505, shapeName=Base_Hair_Front_r_capsule, rotation=<rot 0 -90 0>}]
 Dyn_Hair_Back_01_c [{shapeType=Capsule, length=6, radius=1.5, shapeName=dyn_hair_back_01_c_capsule, center=<3 0 0>, rotation=<rot 90 0 0>}]
 Dyn_Hair_Back_01_l [{shapeType=Capsule, length=6.5, radius=2, shapeName=dyn_hair_back_01_l_capsule, center=<5.1769 0 0>, rotation=<rot 55 -89.9997 90.0004>}]
 Dyn_Hair_Back_01_r [{shapeType=Capsule, length=6.5, radius=2, center=<5.1769 0 0>, rotation=<rot -50 -90.0002 90.0003>}]
 Dyn_Hair_Back_02_c [{shapeType=Capsule, length=3.25, radius=3, center=<3 0 0>, rotation=<rot -0.0007 0 90>}]
 Dyn_Hair_Back_02_l [{
   shapeType=Capsule
   length=5.75
   radius=2
   shapeName=dyn_hair_back_02_l_capsule
   center=<4.2315 0 -0.0277>
   rotation=<rot -5.4432 12.7585 140.5139>
  }]
 Dyn_Hair_Back_02_r [{
   shapeType=Capsule
   length=4.7875
   radius=1.7125
   shapeName=dyn_hair_back_02_r_capsule
   center=<4.892 -0.6403 -1.552>
   rotation=<rot -8.4373 -12.4483 76.3125>
  }]
 Dyn_Hair_Back_03_c [{shapeType=Capsule, length=3.6906, radius=2.5, shapeName=Dyn_Hair_Back_03_c_capsule, center=<3 0 0>, rotation=<rot 90 0 180>}]
 Dyn_Hair_Back_03_l [{
   shapeType=Capsule
   length=9
   radius=1.25
   shapeName=dyn_hair_back_03_l_capsule
   center=<2.4132 1.0671 -1.6725>
   rotation=<rot -19.7829 25.3813 120.712>
  }]
 Dyn_Hair_Back_03_r [{
   shapeType=Capsule
   length=2.8436
   radius=2
   shapeName=dyn_hair_back_03_r_capsule
   center=<3.8625 -1.0873 0.1937>
   rotation=<rot -5.6948 -3.194 89.5002>
  }]
 Dyn_Hair_Back_04_c [{shapeType=Capsule, length=4.1604, radius=1.9569, shapeName=Dyn_Hair_Back_04_c_capsule, center=<4.687 0 0>, rotation=<rot 90 -90 0>}]
 Dyn_Hair_Back_04_l [{shapeType=Capsule, length=6.75, radius=1.25, shapeName=dyn_hair_back_04_l_capsule, center=<3 0 0>, rotation=<rot -0.0001 0 90>}]
 Dyn_Hair_Back_04_r [{shapeType=Capsule, length=5.4041, radius=1.8459, shapeName=dyn_hair_back_04_r_capsule, center=<3.1875 -1.5049 0.6684>, rotation=<rot 0 0 90>}]
 Dyn_Hair_Back_05_c [{shapeType=Capsule, length=2.8021, radius=2, shapeName=dyn_hair_back_05_c_capsule, center=<5.1842 0 -0.6234>, rotation=<rot 79.9997 0 90>}]
 Dyn_Hair_Back_05_l [{shapeType=Capsule, length=3.7413, radius=1.3507, shapeName=Dyn_Hair_Back_05_l_capsule, center=<4.0195 0 0.2489>, rotation=<rot 0 -90 -90>}]
 Dyn_Hair_Front_01_r [{shapeType=Capsule, length=6.5, radius=2, center=<5.1769 0 0>, rotation=<rot -50 -90.0002 90.0003>}]
 Dyn_Hair_Front_02_r [{shapeType=Capsule, length=2, radius=2, shapeName=Dyn_Hair_Front_02_r_capsule, center=<3.0259 0 0>, rotation=<rot 90 0 0>}]
 Dyn_Hair_Front_03_r [{shapeType=Capsule, length=1.8999, radius=1.3474, shapeName=Dyn_Hair_Front_03_r_capsule, center=<2.8954 0.4756 0>, rotation=<rot 90 0 0>}]
 clavicle_l [{shapeType=Capsule, length=3.75, radius=10.5, shapeName=clavicle_l_capsule, center=<4.1135 -1.5355 -3.2101>, rotation=<rot 90 0 0>}]
 clavicle_r [{
   shapeType=Capsule
   length=1.25
   radius=4.75
   shapeName=clavicle_l_capsule
   center=<-3.7685 0.8897 -4.2811>
   rotation=<rot 6.7179 18.8817 -108.8823>
  }]
 head [{shapeType=Capsule, length=3.969, radius=5.5623, center=<1.8699 0 0>, rotation=<rot 90 180 -180>}]
 neck_01 [
  {shapeType=Capsule, length=13.5, radius=3.25, shapeName=neck_01_capsule, center=<6.3954 1.154 0>, rotation=<rot 90 180 175.0004>}
  {shapeType=Capsule, length=21.25, radius=16.5, shapeName=neck_01_capsule, center=<6.7378 15.6571 0>}
 ]
 spine_05 [
  {shapeType=Capsule, length=8, radius=8.75, shapeName=spine_05_capsule, center=<4.9041 -1.2528 0>}
  {shapeType=Capsule, length=2.5, radius=8.5, shapeName=clavicle_r_capsule, center=<5.315 -1.0698 7.814>}
 ]
}
constraints [
 {
  bone1Name=Base_Hair_Back_c
  bone2Name=head
  swing1LimitDegrees=45
  swing1Motion=Free
  swing2LimitDegrees=45
  swing2Motion=Free
  twistLimitDegrees=45
  twistMotion=Free
 }
 {bone1Name=Dyn_Hair_Back_01_c, bone2Name=Base_Hair_Back_c, swing1LimitDegrees=0, swing2LimitDegrees=5, twistLimitDegrees=0}
 {
  bone1Name=Dyn_Hair_Back_02_c
  bone2Name=Dyn_Hair_Back_01_c
  swing1LimitDegrees=5
  swing2LimitDegrees=30
  twistLimitDegrees=0
  twistMotion=Locked
 }
 {bone1Name=Dyn_Hair_Back_03_c, bone2Name=Dyn_Hair_Back_02_c, swing1LimitDegrees=5, swing2LimitDegrees=45, twistLimitDegrees=10}
 {
  bone1Name=Dyn_Hair_Back_04_c
  bone2Name=Dyn_Hair_Back_03_c
  swing1LimitDegrees=10
  swing2LimitDegrees=40
  twistLimitDegrees=45
  twistMotion=Locked
 }
 {
  bone1Name=Dyn_Hair_Back_05_c
  bone2Name=Dyn_Hair_Back_04_c
  swing1LimitDegrees=20
  swing2LimitDegrees=45
  twistLimitDegrees=45
  twistMotion=Locked
 }
 {
  bone1Name=Base_Hair_Back_l
  bone2Name=head
  swing1LimitDegrees=45
  swing1Motion=Free
  swing2LimitDegrees=45
  swing2Motion=Free
  twistLimitDegrees=45
  twistMotion=Free
 }
 {
  bone1Name=Dyn_Hair_Back_02_l
  bone2Name=Dyn_Hair_Back_01_l
  swing1LimitDegrees=5
  swing2LimitDegrees=35
  twistLimitDegrees=45
  twistMotion=Locked
 }
 {bone1Name=Dyn_Hair_Back_03_l, bone2Name=Dyn_Hair_Back_02_l, swing1LimitDegrees=5, swing2LimitDegrees=45, twistLimitDegrees=15}
 {bone1Name=Dyn_Hair_Back_01_l, bone2Name=Base_Hair_Back_l, swing1LimitDegrees=1, swing2LimitDegrees=5, twistLimitDegrees=0}
 {
  bone1Name=Base_Hair_Back_r
  bone2Name=head
  swing1LimitDegrees=45
  swing1Motion=Free
  swing2LimitDegrees=45
  swing2Motion=Free
  twistLimitDegrees=45
  twistMotion=Free
 }
 {bone1Name=Dyn_Hair_Back_01_r, bone2Name=Base_Hair_Back_r, swing1LimitDegrees=10, swing2LimitDegrees=10, twistLimitDegrees=0}
 {bone1Name=Dyn_Hair_Back_02_r, bone2Name=Dyn_Hair_Back_01_r, swing1LimitDegrees=5, swing2LimitDegrees=35, twistLimitDegrees=10}
 {bone1Name=Dyn_Hair_Back_03_r, bone2Name=Dyn_Hair_Back_02_r, swing1LimitDegrees=20, swing2LimitDegrees=45, twistLimitDegrees=20}
 {
  bone1Name=Dyn_Hair_Back_04_r
  bone2Name=Dyn_Hair_Back_03_r
  swing1LimitDegrees=20
  swing2LimitDegrees=35
  twistLimitDegrees=45
  twistMotion=Locked
 }
 {
  bone1Name=head
  bone2Name=spine_05
  swing1LimitDegrees=45
  swing1Motion=Free
  swing2LimitDegrees=45
  swing2Motion=Free
  twistLimitDegrees=45
  twistMotion=Free
 }
 {
  bone1Name=clavicle_l
  bone2Name=spine_05
  swing1LimitDegrees=45
  swing1Motion=Free
  swing2LimitDegrees=45
  swing2Motion=Free
  twistLimitDegrees=45
  twistMotion=Free
 }
 {
  bone1Name=Dyn_Hair_Back_04_l
  bone2Name=Dyn_Hair_Back_03_l
  swing1LimitDegrees=10
  swing2LimitDegrees=35
  twistLimitDegrees=45
  twistMotion=Locked
 }
 {
  bone1Name=Base_Hair_Front_r
  bone2Name=head
  swing1LimitDegrees=45
  swing1Motion=Free
  swing2LimitDegrees=45
  swing2Motion=Free
  twistLimitDegrees=45
  twistMotion=Free
 }
 {
  bone1Name=Dyn_Hair_Front_01_r
  bone2Name=Base_Hair_Front_r
  swing1LimitDegrees=10
  swing2LimitDegrees=10
  twistLimitDegrees=45
  twistMotion=Locked
 }
 {
  bone1Name=Dyn_Hair_Front_02_r
  bone2Name=Dyn_Hair_Front_01_r
  swing1LimitDegrees=15
  swing2LimitDegrees=20
  twistLimitDegrees=45
  twistMotion=Locked
 }
 {
  bone1Name=Dyn_Hair_Front_03_r
  bone2Name=Dyn_Hair_Front_02_r
  swing1LimitDegrees=25
  swing2LimitDegrees=25
  twistLimitDegrees=45
  twistMotion=Locked
 }
 {
  bone1Name=Dyn_Hair_Back_05_l
  bone2Name=Dyn_Hair_Back_04_l
  swing1LimitDegrees=15
  swing2LimitDegrees=25
  twistLimitDegrees=45
  twistMotion=Locked
 }
 {
  bone1Name=clavicle_r
  bone2Name=spine_05
  swing1LimitDegrees=45
  swing1Motion=Free
  swing2LimitDegrees=45
  swing2Motion=Free
  twistLimitDegrees=45
  twistMotion=Free
 }
 {
  bone1Name=Dyn_Hair_Back_04_r
  bone2Name=Dyn_Hair_Back_04_c
  swing1LimitDegrees=45
  swing1Motion=Free
  swing2LimitDegrees=45
  swing2Motion=Free
  twistLimitDegrees=45
  twistMotion=Free
 }
]
