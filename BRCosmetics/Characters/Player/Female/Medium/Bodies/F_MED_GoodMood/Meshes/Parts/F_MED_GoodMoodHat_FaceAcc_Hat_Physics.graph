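# Physics asset for the hat face accessory of the GoodMood medium female player cosmetic, setting up secondary motion for the back of the hat.
# Center and left Dyn_Hat_Back capsules hang from small base capsules under tight swing limits, with twist limited on the center and locked on the left.
# Both bases link to spine_05 with free angular motion, as do the clavicles, whose large capsules form an upper-body volume with the spine.
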PhysicsAsset /BRCosmetics/Characters/Player/Female/Medium/Bodies/F_MED_GoodMood/Meshes/Parts/F_MED_GoodMoodHat_FaceAcc_Hat_Physics
bodies {
 Base_Hat_Back_c [{shapeType=Capsule, length=0.505, radius=1, shapeName=Base_Hat_Back_c_capsule, rotation=<rot 0 -90 0>}]
 Base_Hat_Back_l [{shapeType=Capsule, length=0.505, radius=0.505, shapeName=Base_Hat_Back_l_capsule, rotation=<rot 0 -90 0>}]
 Dyn_Hat_Back_01_c [{
   shapeType=Capsule
   length=12.3649
   radius=6.2325
   shapeName=Dyn_Hat_Back_01_c_capsule
   center=<12.8943 0 -8.8857>
   rotation=<rot 0.0001 180 -90>
  }]
 Dyn_Hat_Back_l [{
   shapeType=Capsule
   length=4.3336
   radius=7.0577
   shapeName=Dyn_Hat_Back_l_capsule
   center=<13.992 -3.6673 -6.209>
   rotation=<rot 0 -8.3676 -68.8015>
  }]
 clavicle_l [{shapeType=Capsule, length=3.0879, radius=8.4177, shapeName=clavicle_r_capsule, center=<5.7473 0.4551 -2.5744>, rotation=<rot -0.0001 -90 95>}]
 clavicle_r [{
   shapeType=Capsule
   length=3.6566
   radius=7.8491
   shapeName=clavicle_r_capsule
   center=<-5.7473 -0.4551 2.5744>
   rotation=<rot -0.0001 -90 -85>
  }]
 spine_05 [{
   shapeType=Capsule
   length=1.1405
   radius=13.8433
   shapeName=spine_05_capsule
   center=<3.7108 1.2692 0.0119>
   rotation=<rot -0.158 -90 0.0616>
  }]
 upperarm_r [{shapeType=Capsule, length=0.505, radius=0.505, shapeName=upperarm_r_capsule, rotation=<rot 0 -90 0>}]
}
constraints [
 {bone1Name=Dyn_Hat_Back_01_c, bone2Name=Base_Hat_Back_c, swing1LimitDegrees=5, swing2LimitDegrees=20, twistLimitDegrees=10}
 {
  bone1Name=Base_Hat_Back_c
  bone2Name=spine_05
  swing1LimitDegrees=45
  swing1Motion=Free
  swing2LimitDegrees=45
  swing2Motion=Free
  twistLimitDegrees=45
  twistMotion=Free
 }
 {
  bone1Name=upperarm_r
  bone2Name=clavicle_r
  swing1LimitDegrees=45
  swing1Motion=Free
  swing2LimitDegrees=45
  swing2Motion=Free
  twistLimitDegrees=45
  twistMotion=Free
 }
 {
  bone1Name=clavicle_l
  bone2Name=spine_05
  swing1LimitDegrees=45
  swing1Motion=Free
  swing2LimitDegrees=45
  swing2Motion=Free
  twistLimitDegrees=45
  twistMotion=Free
 }
 {
  bone1Name=clavicle_r
  bone2Name=spine_05
  swing1LimitDegrees=45
  swing1Motion=Free
  swing2LimitDegrees=45
  swing2Motion=Free
  twistLimitDegrees=45
  twistMotion=Free
 }
 {bone1Name=Dyn_Hat_Back_l, bone2Name=Base_Hat_Back_l, swing1LimitDegrees=5, swing2LimitDegrees=20, twistLimitDegrees=45, twistMotion=Locked}
 {
  bone1Name=Base_Hat_Back_l
  bone2Name=spine_05
  swing1LimitDegrees=45
  swing1Motion=Free
  swing2LimitDegrees=45
  swing2Motion=Free
  twistLimitDegrees=45
  twistMotion=Free
 }
]
