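# Post-process-style character part AnimBP for the female GoodMood hat/face accessory, adding secondary motion to earrings, feathers, hat and hair.
# Copied parent and body-mesh poses feed a Control Rig, then cached chains of AnimDynamics, additive bone tweaks and hat RigidBody.
# A final RigidBody blend selects backpack or free physics assets by back bling presence.
# Updates drive gravity override, skydive-dependent inertia scales and head-relative hat lift.
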AnimBlueprint /BRCosmetics/Characters/Player/Female/Medium/Bodies/F_MED_GoodMood/Meshes/Parts/F_MED_GoodMoodHat_FaceAcc_AnimBP
target_skeleton=&/F_MED_GoodMoodHat_FaceAcc_Skeleton
graphs [
 AnimGraph {nodes [
   n1 = Get GravityOverride()
   n2 = RigidBody!(ComponentPose=Local To Component(LocalPose=Use cached pose 'HairCache'()), Alpha, OverrideWorldGravity:Vector=@n1, ExternalForce) -> (Pose) {
    bClampLinearTranslationLimitToRefPose=true
    bOverrideWorldGravity=true
    overridePhysicsAsset=&/F_MED_GoodMoodHat_FaceAcc_BackPack_Physics
    simulationSpace=BaseBoneSpace
    alphaBoolBlend {blendOption=Linear}
    alphaScaleBiasClamp {bInterpResult=true, interpSpeedIncreasing=25}
    baseBoneRef {boneName=neck_01}
    componentAppliedLinearAccClamp=<3000 3000 3000>
    componentLinearAccScale=<0.1 0.2 0.2>
    componentLinearVelScale=<0.1 0.2 0.2>
    overrideWorldGravity {z=-780}
   }
   n3 = Get Linear Scale Body() -> (Linear Scale Body:Vector)
   n4 = RigidBody!(
    ComponentPose=Local To Component(LocalPose=Use cached pose 'HairCache'())
    Alpha
    OverrideWorldGravity=@n1
    ExternalForce
    ComponentLinearAccScale:Vector=@n3
    ComponentLinearVelScale:Vector=@n3
   ) -> (Pose) {
    bClampLinearTranslationLimitToRefPose=true
    bOverrideWorldGravity=true
    overridePhysicsAsset=&/F_MED_GoodMoodHat_FaceAcc_Free_Physics
    simulationSpace=BaseBoneSpace
    alphaBoolBlend {blendOption=Linear}
    alphaScaleBiasClamp {bInterpResult=true, interpSpeedIncreasing=25}
    baseBoneRef {boneName=head}
    componentAppliedLinearAccClamp=<3000 3000 3000>
    componentLinearAccScale=<0.45 0.3 0.2>
    componentLinearVelScale=<0.4 0.3 0.2>
    overrideWorldGravity {z=-780}
   }
   Output Pose(Result=Blend Poses by bool(
     BlendPose_0=Component To Local(ComponentPose=@n2)
     BlendPose_1=Component To Local(ComponentPose=@n4)
     bActiveValue=Get IsWearing() -> (IsWearing:bool)
    ) {blendType=Linear}) {layerGroup=None}
   n5 = Anim Dynamics - Bone: Dyn_Fether_01(ComponentPose=Local To Component(LocalPose=Use cached pose 'EaringCache'()), Alpha=1, ExternalForce="0,0,0") {
    angularSpringConstant=200
    bAngularSpring=true
    alphaBoolBlend {blendOption=Linear}
    boundBone {boneName=Dyn_Fether_01}
    componentLinearAccScale {y=0.3, z=0.3}
    componentLinearVelScale {y=0.3, z=0.3}
    physicsBodyDefinitions [{
      boundBone {boneName=Dyn_Fether_01}
      boxExtents=<50 50 50>
      constraintSetup {angularLimitsMax=<0 30 0>, angularLimitsMin=<0 0 0>, angularTarget=<0 1 0>}
      localJointOffset=<0 5 0>
     }]
   }
   n6 = Anim Dynamics - Bone: Dyn_Fether_02(ComponentPose=@n5, Alpha=1, ExternalForce="0,0,0") {
    angularSpringConstant=200
    bAngularSpring=true
    alphaBoolBlend {blendOption=Linear}
    boundBone {boneName=Dyn_Fether_02}
    componentLinearAccScale {y=0.3, z=0.3}
    componentLinearVelScale {y=0.3, z=0.3}
    physicsBodyDefinitions [{
      boundBone {boneName=Dyn_Fether_02}
      boxExtents=<10 3 10>
      constraintSetup {angularLimitsMax=<0 30 0>, angularLimitsMin=<0 0 0>, angularTarget=<0 1 0>}
      localJointOffset=<0 1 0>
     }]
   }
   n7 = Get Hair Up_1() -> (Hair Up_1:double)
   n8 = "Transform (Modify) Bone - Bone: Base_Hat_Back_l"(ComponentPose=@n6, Alpha=@n7, Translation="0,-2,0", Rotation="35,0,0", Scale="1,1,1") {
    rotationMode=BMM_Additive
    rotationSpace=BCS_BoneSpace
    translationMode=BMM_Additive
    alphaBoolBlend {blendOption=CubicInOut, blendOutTime=10}
    boneToModify {boneName=Base_Hat_Back_l}
    rotation {pitch=35}
    translation {y=-2}
   }
   n9 = "Transform (Modify) Bone - Bone: Base_Hat_Back_c"(ComponentPose=@n8, Alpha=@n7, Translation="0,0,0", Rotation="40,1.22,10.251", Scale="1,1,1") {
    rotationMode=BMM_Additive
    rotationSpace=BCS_BoneSpace
    translationMode=BMM_Additive
    alphaBoolBlend {blendOption=CubicInOut, blendOutTime=10}
    boneToModify {boneName=Base_Hat_Back_c}
    rotation=<rot 40 1.22 10.251>
   }
   n10 = Anim Dynamics - Bone: Dyn_Hair_Front_01_c(ComponentPose=@n9, Alpha=1, ExternalForce="0,0,0") {
    angularSpringConstant=5000
    gravityScale=0.7
    alphaBoolBlend {blendOption=Linear}
    boundBone {boneName=Dyn_Hair_Front_01_c}
    componentLinearAccScale=<0.5 0.5 0.5>
    componentLinearVelScale=<0.5 0.5 0.5>
    physicsBodyDefinitions [{
      boundBone {boneName=Dyn_Hair_Front_01_c}
      boxExtents=<20 20 10>
      constraintSetup {angularLimitsMax=<0 5 0>, angularLimitsMin=<0 -30 0>, angularTarget=<1 0 0>}
      localJointOffset=<10 0 0>
     }]
   }
   n11 = Get GravityOverride()
   n12 = Anim Dynamics - Bone: Dyn_Hat_01_l!(ComponentPose=@n10, Alpha=0.4, GravityOverride:Vector=@n11, ExternalForce="0,0,0") -> (Pose) {
    angularSpringConstant=5000
    bUseGravityOverride=true
    gravityScale=0.7
    alphaBoolBlend {blendOption=Linear}
    boundBone {boneName=Dyn_Hat_01_l}
    componentAppliedLinearAccClamp=<5000 5000 5000>
    componentLinearAccScale=<0.5 0.5 0.4>
    componentLinearVelScale=<0.5 0.5 0.4>
    physicsBodyDefinitions [{
      boundBone {boneName=Dyn_Hat_01_l}
      boxExtents=<20 20 10>
      constraintSetup {angularLimitsMax=<0 5 0>, angularLimitsMin=<0 -30 0>, angularTarget=<1 0 0>}
      localJointOffset=<10 0 0>
     }]
   }
   n13 = Anim Dynamics - Bone: Dyn_Hat_01_r!(ComponentPose=@n12, Alpha=0.4, GravityOverride:Vector=@n11, ExternalForce="0,0,0") -> (Pose) {
    angularSpringConstant=5000
    bUseGravityOverride=true
    gravityScale=0.7
    simulationSpace=BoneRelative
    alphaBoolBlend {blendOption=Linear}
    boundBone {boneName=Dyn_Hat_01_r}
    componentAppliedLinearAccClamp=<5000 5000 5000>
    componentLinearAccScale=<0.5 0.5 0.4>
    componentLinearVelScale=<0.5 0.5 0.4>
    physicsBodyDefinitions [{
      boundBone {boneName=Dyn_Hat_01_r}
      boxExtents=<20 20 10>
      constraintSetup {angularLimitsMax=<0 5 0>, angularLimitsMin=<0 -20 0>, angularTarget=<1 0 0>}
      localJointOffset=<10.9471 0 0.8157>
     }]
    relativeSpaceBone {boneName=Base_Hat_r}
   }
   n14 = Get Linear Scale Hat() -> (Linear Scale Hat:Vector)
   n15 = RigidBody(ComponentPose=@n13, OverrideWorldGravity=Get GravityOverride(), ComponentLinearAccScale=@n14, ComponentLinearVelScale=@n14) {
    bClampLinearTranslationLimitToRefPose=true
    bOverrideWorldGravity=true
    overridePhysicsAsset=&/F_MED_GoodMoodHat_FaceAcc_Hat_Physics
    simulationSpace=BaseBoneSpace
    alphaBoolBlend {blendOption=Linear}
    baseBoneRef {boneName=head}
    componentLinearAccScale=<0.4 0.2 0.2>
    componentLinearVelScale=<0.35 0.2 0.2>
   }
   n16 = "Transform (Modify) Bone - Bone: Base_Hat_Back_c"(ComponentPose=@n15, Alpha=Get skydiving float() -> (skydiving float:double), Translation="0,0,0", Rotation="10,0,-15", Scale="1,1,1") {
    rotationMode=BMM_Additive
    rotationSpace=BCS_BoneSpace
    translationMode=BMM_Additive
    translationSpace=BCS_BoneSpace
    alphaBoolBlend {blendOption=CubicInOut, blendOutTime=10}
    boneToModify {boneName=Base_Hat_Back_c}
    rotation {pitch=10, roll=-15}
   }
   Save cached pose 'Hat Cache'(Pose=Component To Local(ComponentPose=@n16)) {cacheName=Hat Cache}
   n17 = Use cached pose 'Hat Cache'()
   Save cached pose 'HairCache'(Pose=Component To Local(ComponentPose=Local To Component(LocalPose=@n17))) {cacheName=HairCache}
   n18 = Layered blend per bone(
    BasePose=Copy Pose From Mesh() {bCopyCurves=true, bUseAttachedParent=true}
    BlendPoses_0=Copy Pose From Mesh(SourceMeshComponent=Get Body_Mesh())
   ) {layerSetup [{branchFilters [{boneName=clavicle_l}, {boneName=clavicle_r}, {boneName=upperarm_r}, {boneName=upperarm_l}]}]}
   n19 = Control Rig(Source=@n18) {
    alphaBoolBlend {blendOption=Linear}
    controlRigAssetReference {blueprintRigClass=&/F_MED_GoodMoodHat_FaceAcc_CtrlRig.F_MED_GoodMoodHat_FaceAcc_CtrlRig_C}
   }
   Save cached pose 'StartPose'(Pose=@n19) {cacheName=StartPose}
   n20 = Anim Dynamics - Bone: Dyn_Earing_01_l(ComponentPose=Local To Component(LocalPose=Use cached pose 'StartPose'()), Alpha=1, ExternalForce="0,0,0") {
    angularSpringConstant=300
    bAngularSpring=true
    bUseSphericalLimits=true
    alphaBoolBlend {blendOption=Linear}
    boundBone {boneName=Dyn_Earing_01_l}
    componentLinearAccScale=<0.1 0.1 0.1>
    componentLinearVelScale=<0.1 0.1 0.1>
    physicsBodyDefinitions [{
      boundBone {boneName=Dyn_Earing_01_l}
      boxExtents=<30 30 20>
      constraintSetup {angularLimitsMax=<0 40 0>, angularLimitsMin=<0 -40 0>, angularTarget=<1 0 0>}
      localJointOffset=<2 0 0>
     }]
   }
   n21 = Anim Dynamics - Bone: Dyn_Earing_02_l(ComponentPose=@n20, Alpha=1, ExternalForce="0,0,0") {
    angularSpringConstant=10000
    bAngularSpring=true
    bUseSphericalLimits=true
    alphaBoolBlend {blendOption=Linear}
    boundBone {boneName=Dyn_Earing_02_l}
    componentLinearAccScale=<0.1 0.1 0.1>
    componentLinearVelScale=<0.1 0.1 0.1>
    physicsBodyDefinitions [{
      boundBone {boneName=Dyn_Earing_02_l}
      boxExtents=<40 40 20>
      constraintSetup {angularLimitsMax=<0 10 0>, angularLimitsMin=<0 -10 -10>, angularTarget=<1 0 0>}
      localJointOffset=<2 0 0>
     }]
   }
   n22 = Anim Dynamics - Bone: Dyn_Earing_01_r(ComponentPose=@n21, Alpha=1, ExternalForce="0,0,0") {
    angularSpringConstant=300
    bAngularSpring=true
    bUseSphericalLimits=true
    alphaBoolBlend {blendOption=Linear}
    boundBone {boneName=Dyn_Earing_01_r}
    componentLinearAccScale=<0.1 0.1 0.1>
    componentLinearVelScale=<0.1 0.1 0.1>
    physicsBodyDefinitions [{
      boundBone {boneName=Dyn_Earing_01_r}
      boxExtents=<30 30 20>
      constraintSetup {angularLimitsMax=<0 40 0>, angularLimitsMin=<0 -40 0>, angularTarget=<1 0 0>}
      localJointOffset=<2 0 0>
     }]
   }
   n23 = Anim Dynamics - Bone: Dyn_Earing_02_r(ComponentPose=@n22, Alpha=1, ExternalForce="0,0,0") {
    angularSpringConstant=10000
    bAngularSpring=true
    bUseSphericalLimits=true
    alphaBoolBlend {blendOption=Linear}
    boundBone {boneName=Dyn_Earing_02_r}
    componentLinearAccScale=<0.1 0.1 0.1>
    componentLinearVelScale=<0.1 0.1 0.1>
    physicsBodyDefinitions [{
      boundBone {boneName=Dyn_Earing_02_r}
      boxExtents=<40 40 20>
      constraintSetup {angularLimitsMax=<0 10 0>, angularLimitsMin=<0 -10 -10>, angularTarget=<1 0 0>}
      localJointOffset=<2 0 0>
     }]
   }
   Save cached pose 'EaringCache'(Pose=Component To Local(ComponentPose=@n23)) {cacheName=EaringCache}
  ]}
 EventGraph {nodes [
   Event OnCharacterCustomizationLoaded()
   > n1 = Cast To PlayerPawn_Athena(Object=TryGetPawnOwner())
   > Set Body_Mesh(Body_Mesh=GetSkeletalMeshForPartType(self=@n1))
   n2 = CheckForBackbling()
   > n3 = Set isWearingBackbling(isWearingBackbling=AND Boolean!(A=@n2, B=NOT Boolean(A=@n2.bIsBackblingACape), C:bool=NOT Boolean(A=@n2.bIsBackblingLong)) -> (ReturnValue))
   > n4 = Set Is Wearing Cape(Is Wearing Cape=OR Boolean(A=@n2.bIsBackblingACape, B=@n2.bIsBackblingLong))
   > Set IsWearing(IsWearing:bool=OR Boolean(A=@n4, B=@n3)) -> (Output_Get:bool)
   Event BlueprintInitializeAnimation()
   > Set Skel Mesh(Skel Mesh=GetOwnerSkeletalMeshComponent())
   n5 = Event BlueprintUpdateAnimation()
   > n6 = Sequence!(execute) -> (then_0, then_1, then_2:Exec, then_3:Exec, then_4:Exec, then_5:Exec, then_6:Exec, then_7:Exec)
   n7 = Branch(execute=@n6.then_1, Condition=Get bIsSkydiveFloating())
   > Set Linear Scale Body(Linear Scale Body:Vector="0.3,0.2,0.2") -> (Output_Get:Vector)
   > Set Linear Scale Backpack(Linear Scale Backpack:Vector="0.1,0,0") -> (Output_Get:Vector)
   > Set Linear Scale Hat(Linear Scale Hat:Vector="0.3,0.2,0.2") -> (Output_Get:Vector)
   Set Linear Scale Body(execute=@n7.else, Linear Scale Body="0.34,0.2,0.2")
   > Set Linear Scale Backpack(Linear Scale Backpack="0.2,0.1,0.1")
   > Set Linear Scale Hat(Linear Scale Hat="0.4,0.2,0.2")
   n8 = AND Boolean!(A=Get bIsMoving2D(), B=Get bIsCrouching()) -> (ReturnValue)
   n9 = "To Float (Boolean)"(InBool=@n8)
   Set HeadUpCrouch(execute=@n6.then_3, HeadUpCrouch:double=float * float(A=Get Hair Up() -> (Hair Up:double), B=@n9) {is=title K2Node_PromotableOperator}) -> (Output_Get:double)
   n10 = AND Boolean(A=@n8, B=Get bIsTargeting())
   Set HeadTargetCrouch(
    execute=@n6.then_4
    HeadTargetCrouch:double=FInterpTo(
     Current=Get HeadTargetCrouch() -> (HeadTargetCrouch:double)
     Target="To Float (Boolean)"(InBool=@n10)
     DeltaTime=@n5.DeltaTimeX
     InterpSpeed=8
    )
   ) -> (Output_Get:double)
   n11 = PlayerMovement(execute=@n6.then_5, AnimInstance=Self-Reference(), XDotProductCoefficient=100, YDotProductCoefficient=100, ZDotProductCoefficient=100)
   n12 = BreakVector(InVec=MakeVector(X=@n11, Y=@n11.YMovement))
   > Set Left(Left:double="To Float (Boolean)"(InBool="float < float"(A=@n12, B=-50))) -> (Output_Get:double)
   > Set Right(Right:double="To Float (Boolean)"(InBool="float > float"(A=@n12, B=50))) -> (Output_Get:double)
   > Set Forward(Forward:double="To Float (Boolean)"(InBool="float > float"(A=@n12.Y, B=40))) -> (Output_Get:double)
   > Set Backward(Backward:double="To Float (Boolean)"(InBool="float < float"(A=@n12.Y, B=-40))) -> (Output_Get:double)
   Set CrouchWalk(execute=@n6.then_6, CrouchWalk:double=@n9) -> (Output_Get:double)
   n13 = Get Skel Mesh()
   Set Hair Up(
    execute=@n6.then_2
    Hair Up:double=float * float(
     A=GetAlphaFromReferenceTransformDelta(SkeletalMeshComponent=@n13, BoneName=spine_05, BaseName=head, InputRange=(X=0,Y=90), TwistAxis="0,0,1")
     B=1
    )
   ) -> (Output_Get:double)
   > Set Hair Up_1(Hair Up_1:double=GetAlphaFromReferenceTransformDelta(SkeletalMeshComponent=Get Skel Mesh(), BoneName=spine_05, BaseName=head, DeltaComponentType=EulerY, InputRange=(X=0,Y=90), TwistAxis="0,0,1")) -> (Output_Get:double)
   n14 = Gravity Override State Function DEPRECATED(
    execute=@n6.then_0
    CharacterPartAnimInstance=Self-Reference()
    GameTimeInSeconds=GetGameTimeInSeconds()
    GravityOverrideParams=Get Gravity Override Params()
   )
   > Set GravityOverride(GravityOverride=@n14)
   > Set skydiving float(skydiving float:double="To Float (Boolean)"(InBool=Get bIsSkydiveFloating())) -> (Output_Get:double)
   Get bIsTargeting()
  ]}
]
variables [
 Backward {type=number}
 Body_Mesh {type=SkeletalMeshComponent}
 CrouchWalk {type=number}
 Forward {type=number}
 Gravity Override Params {
  type=GravityOverrideParamsStruct
  default {skydive_Parachuting {gravityOverride=<0 -800 -1200>, windAmplitude=<200 -500 100>, windFrequency=<1.5 1 1>}}
 }
 GravityOverride {type=Vector}
 Hair Up {type=number}
 Hair Up_0 {type=number}
 Hair Up_1 {type=number}
 HeadTargetCrouch {type=number}
 HeadUpCrouch {type=number}
 HeadUpCrouch_0 {type=number}
 Is Skydive Floating {type=number}
 Is Wearing Cape {type=boolean}
 IsWearing {type=boolean}
 isWearingBackbling {type=boolean}
 Left {type=number}
 Linear Scale Backpack {type=Vector}
 Linear Scale Body {type=Vector}
 Linear Scale Hat {type=Vector}
 Right {type=number}
 Skel Mesh {type=SkeletalMeshComponent}
 skydiving float {type=number}
]
