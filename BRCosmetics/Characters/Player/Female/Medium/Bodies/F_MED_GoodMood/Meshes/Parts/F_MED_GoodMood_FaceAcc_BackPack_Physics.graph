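# Physics asset for the GoodMood medium female character's FaceAcc BackPack mesh part, set up for secondary motion of hair.
# Capsule bodies form center, left and right back hair chains plus a right front strand, each hung from a base bone on the head.
# Back chain roots are nearly locked and lower links swing wider, with twist mostly locked.
# Head, neck, spine, clavicle and right upper arm capsules, including oversized offset ones on neck and upper spine, supply body volumes.
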PhysicsAsset /BRCosmetics/Characters/Player/Female/Medium/Bodies/F_MED_GoodMood/Meshes/Parts/F_MED_GoodMood_FaceAcc_BackPack_Physics
bodies {
 base_hair_back_c [{shapeType=Capsule, length=2, radius=2, shapeName=base_hair_back_c_capsule}]
 base_hair_back_l [{shapeType=Capsule, length=2, radius=2, shapeName=base_hair_back_l_capsule}]
 base_hair_back_r [{shapeType=Capsule, length=2, radius=2, shapeName=base_hair_back_r_capsule}]
 base_hair_front_r [{shapeType=Capsule, length=2, radius=2, shapeName=base_hair_front_r_capsule}]
 clavicle_l [{shapeType=Capsule, length=1, radius=9.75, shapeName=clavicle_l_capsule, center=<5.5885 -2.7361 -2.2822>, rotation=<rot 90 0 -0.0002>}]
 clavicle_r [{shapeType=Capsule, length=2.75, radius=8.25, shapeName=clavicle_r_capsule, center=<-4.5027 -0.9411 0.062>, rotation=<rot 90 0 0>}]
 dyn_hair_back_01_c [{shapeType=Capsule, length=6, radius=2.5, shapeName=dyn_hair_back_01_c_capsule, center=<5.1769 0 0>, rotation=<rot 0 -90 90.0002>}]
 dyn_hair_back_01_l [{shapeType=Capsule, length=6.5, radius=2, shapeName=dyn_hair_back_01_l_capsule, center=<5.1769 0 0>, rotation=<rot 55 -89.9997 90.0004>}]
 dyn_hair_back_01_r [{shapeType=Capsule, length=6.5, radius=2, center=<5.1769 0 0>, rotation=<rot -50 -90.0002 90.0003>}]
 dyn_hair_back_02_c [{shapeType=Capsule, length=2.5, radius=2, center=<3.6105 0 0>, rotation=<rot 0 90 89.9998>}]
 dyn_hair_back_02_l [{shapeType=Capsule, length=3.25, radius=2, shapeName=dyn_hair_back_02_l_capsule, center=<2 0 0>, rotation=<rot -0.0001 0 140>}]
 dyn_hair_back_02_r [{shapeType=Capsule, length=4.5, radius=2, shapeName=dyn_hair_back_02_r_capsule, center=<2 0 0>, rotation=<rot -0.0006 0 75>}]
 dyn_hair_back_03_c [{shapeType=Capsule, length=5, radius=2, shapeName=dyn_hair_back_03_c_capsule, center=<2.5859 0 -1>, rotation=<rot 0 0 90>}]
 dyn_hair_back_03_l [{shapeType=Capsule, length=5.25, radius=2.25, shapeName=dyn_hair_back_03_l_capsule, center=<2.1648 0 0>, rotation=<rot -0.0001 0 110>}]
 dyn_hair_back_03_r [{shapeType=Capsule, length=5.7093, radius=1.7907, shapeName=dyn_hair_back_03_r_capsule, center=<4 0 0>, rotation=<rot -5.7163 -8.2189 90>}]
 dyn_hair_back_04_c [{shapeType=Capsule, length=6, radius=1.5, shapeName=dyn_hair_back_04_c_capsule, center=<3.5859 0 0>, rotation=<rot 0 0 90>}]
 dyn_hair_back_04_l [{shapeType=Capsule, length=4.0916, radius=1.4084, shapeName=dyn_hair_back_04_l_capsule, center=<3 0 0>, rotation=<rot -0.0001 0 90>}]
 dyn_hair_back_04_r [{shapeType=Capsule, length=3.6486, radius=1.8514, shapeName=dyn_hair_back_04_r_capsule, center=<3.2997 0 0>, rotation=<rot 0 0 90>}]
 dyn_hair_back_05_c [{shapeType=Capsule, length=1.4214, radius=2, shapeName=dyn_hair_back_05_c_capsule, center=<5.4576 0 -0.7672>, rotation=<rot 79.9999 0 0>}]
 dyn_hair_front_01_r [{shapeType=Capsule, length=3, radius=1.5, shapeName=dyn_hair_front_01_r_capsule, center=<3 0 0>, rotation=<rot 90 0 0>}]
 dyn_hair_front_02_r [{shapeType=Capsule, length=3, radius=1.5, shapeName=dyn_hair_front_02_r_capsule, center=<3 0 0>, rotation=<rot 90 0 0>}]
 dyn_hair_front_03_r [{shapeType=Capsule, length=3.75, radius=1.5, shapeName=dyn_hair_front_03_r_capsule, center=<3.5351 0 0>, rotation=<rot 90 0 0>}]
 head [{shapeType=Capsule, length=3.75, radius=5.75, center=<3.7841 1.8995 0>, rotation=<rot 90 179.9998 180>}]
 neck_01 [
  {shapeType=Capsule, length=11.5, radius=5.25, shapeName=neck_01_capsule, center=<6.5502 -0.616 0>, rotation=<rot 90 -9.6282 -14.6278>}
  {shapeType=Capsule, length=21.25, radius=16.5, shapeName=neck_01_capsule, center=<6.7378 14.7178 0>}
 ]
 spine_04 [{shapeType=Capsule, length=6.75, radius=8.25, shapeName=spine_04_capsule, rotation=<rot 0 -0.0002 0>}]
 spine_05 [
  {shapeType=Capsule, length=10.75, radius=9.5, shapeName=spine_05_capsule, center=<4.9041 0.179 0>}
  {shapeType=Capsule, length=71.25, radius=15, shapeName=spine_05_capsule, center=<0 -29.6232 0>}
  {shapeType=Capsule, length=71.25, radius=15, shapeName=spine_05_capsule, center=<25.7427 -26.7638 0>}
 ]
 upperarm_r [{shapeType=Capsule, length=22.75, radius=4.75, shapeName=upperarm_r_capsule, center=<-11.1242 0 0.6678>, rotation=<rot 90 0 0>}]
}
constraints [
 {
  bone1Name=dyn_hair_back_02_c
  bone2Name=dyn_hair_back_01_c
  swing1LimitDegrees=5
  swing2LimitDegrees=30
  twistLimitDegrees=45
  twistMotion=Locked
 }
 {
  bone1Name=dyn_hair_back_01_r
  bone2Name=base_hair_back_r
  swing1LimitDegrees=0
  swing1Motion=Locked
  swing2LimitDegrees=5
  twistLimitDegrees=0
  twistMotion=Locked
 }
 {
  bone1Name=upperarm_r
  bone2Name=clavicle_r
  swing1LimitDegrees=45
  swing1Motion=Free
  swing2LimitDegrees=45
  swing2Motion=Free
  twistLimitDegrees=45
  twistMotion=Free
 }
 {
  bone1Name=head
  bone2Name=neck_01
  swing1LimitDegrees=45
  swing1Motion=Free
  swing2LimitDegrees=45
  swing2Motion=Free
  twistLimitDegrees=45
  twistMotion=Free
 }
 {
  bone1Name=neck_01
  bone2Name=spine_05
  swing1LimitDegrees=45
  swing1Motion=Free
  swing2LimitDegrees=45
  swing2Motion=Free
  twistLimitDegrees=45
  twistMotion=Free
 }
 {
  bone1Name=clavicle_l
  bone2Name=spine_05
  swing1LimitDegrees=45
  swing1Motion=Free
  swing2LimitDegrees=45
  swing2Motion=Free
  twistLimitDegrees=45
  twistMotion=Free
 }
 {
  bone1Name=clavicle_r
  bone2Name=spine_05
  swing1LimitDegrees=45
  swing1Motion=Free
  swing2LimitDegrees=45
  swing2Motion=Free
  twistLimitDegrees=45
  twistMotion=Free
 }
 {
  bone1Name=spine_05
  bone2Name=spine_04
  swing1LimitDegrees=45
  swing1Motion=Free
  swing2LimitDegrees=45
  swing2Motion=Free
  twistLimitDegrees=45
  twistMotion=Free
 }
 {
  bone1Name=base_hair_back_l
  bone2Name=head
  swing1LimitDegrees=45
  swing1Motion=Free
  swing2LimitDegrees=45
  swing2Motion=Free
  twistLimitDegrees=45
  twistMotion=Free
 }
 {
  bone1Name=dyn_hair_back_01_l
  bone2Name=base_hair_back_l
  swing1LimitDegrees=0
  swing1Motion=Locked
  swing2LimitDegrees=5
  twistLimitDegrees=0
  twistMotion=Locked
 }
 {
  bone1Name=dyn_hair_back_02_l
  bone2Name=dyn_hair_back_01_l
  swing1LimitDegrees=5
  swing2LimitDegrees=25
  twistLimitDegrees=45
  twistMotion=Locked
 }
 {bone1Name=dyn_hair_back_03_l, bone2Name=dyn_hair_back_02_l, swing1LimitDegrees=20, swing2LimitDegrees=30, twistLimitDegrees=5}
 {
  bone1Name=dyn_hair_back_04_l
  bone2Name=dyn_hair_back_03_l
  swing1LimitDegrees=20
  swing2LimitDegrees=30
  twistLimitDegrees=45
  twistMotion=Locked
 }
 {
  bone1Name=base_hair_back_r
  bone2Name=head
  swing1LimitDegrees=45
  swing1Motion=Free
  swing2LimitDegrees=45
  swing2Motion=Free
  twistLimitDegrees=45
  twistMotion=Free
 }
 {
  bone1Name=dyn_hair_back_02_r
  bone2Name=dyn_hair_back_01_r
  swing1LimitDegrees=5
  swing2LimitDegrees=35
  twistLimitDegrees=45
  twistMotion=Locked
 }
 {
  bone1Name=dyn_hair_back_03_r
  bone2Name=dyn_hair_back_02_r
  swing1LimitDegrees=15
  swing2LimitDegrees=30
  twistLimitDegrees=45
  twistMotion=Locked
 }
 {
  bone1Name=dyn_hair_back_04_r
  bone2Name=dyn_hair_back_03_r
  swing1LimitDegrees=20
  swing2LimitDegrees=30
  twistLimitDegrees=45
  twistMotion=Locked
 }
 {
  bone1Name=base_hair_front_r
  bone2Name=head
  swing1LimitDegrees=45
  swing1Motion=Free
  swing2LimitDegrees=45
  swing2Motion=Free
  twistLimitDegrees=45
  twistMotion=Free
 }
 {
  bone1Name=dyn_hair_front_01_r
  bone2Name=base_hair_front_r
  swing1LimitDegrees=10
  swing2LimitDegrees=10
  twistLimitDegrees=45
  twistMotion=Locked
 }
 {
  bone1Name=dyn_hair_front_02_r
  bone2Name=dyn_hair_front_01_r
  swing1LimitDegrees=10
  swing2LimitDegrees=10
  twistLimitDegrees=45
  twistMotion=Locked
 }
 {
  bone1Name=dyn_hair_front_03_r
  bone2Name=dyn_hair_front_02_r
  swing1LimitDegrees=15
  swing2LimitDegrees=10
  twistLimitDegrees=45
  twistMotion=Locked
 }
 {
  bone1Name=dyn_hair_back_03_c
  bone2Name=dyn_hair_back_02_c
  swing1LimitDegrees=10
  swing2LimitDegrees=30
  twistLimitDegrees=45
  twistMotion=Locked
 }
 {
  bone1Name=dyn_hair_back_04_c
  bone2Name=dyn_hair_back_03_c
  swing1LimitDegrees=20
  swing2LimitDegrees=30
  twistLimitDegrees=45
  twistMotion=Locked
 }
 {
  bone1Name=dyn_hair_back_05_c
  bone2Name=dyn_hair_back_04_c
  swing1LimitDegrees=45
  swing2LimitDegrees=45
  twistLimitDegrees=45
  twistMotion=Locked
 }
 {
  bone1Name=base_hair_back_c
  bone2Name=head
  swing1LimitDegrees=45
  swing1Motion=Free
  swing2LimitDegrees=45
  swing2Motion=Free
  twistLimitDegrees=45
  twistMotion=Free
 }
 {
  bone1Name=dyn_hair_back_01_c
  bone2Name=base_hair_back_c
  swing1LimitDegrees=0
  swing1Motion=Locked
  swing2LimitDegrees=5
  twistLimitDegrees=0
  twistMotion=Locked
 }
]
