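# Physics asset for the F_MED_GoodMood female cosmetic's FaceAcc_Free part, setting up physics for dangling hair strands.
# Capsule bodies form back-center, back-left, back-right and front-right hair chains off the head, alongside neck, upper spine and clavicle bodies.
# Base and body joints have free angular motion, while back-hair roots are nearly locked.
# Swing limits generally widen toward the tips, with twist mostly locked.
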PhysicsAsset /BRCosmetics/Characters/Player/Female/Medium/Bodies/F_MED_GoodMood/Meshes/Parts/F_MED_GoodMood_FaceAcc_Free_Physics
bodies {
 base_hair_back_c [{shapeType=Capsule, length=2, radius=2, shapeName=base_hair_back_c_capsule}]
 base_hair_back_l [{shapeType=Capsule, length=2, radius=2, shapeName=base_hair_back_l_capsule}]
 base_hair_back_r [{shapeType=Capsule, length=2, radius=2, shapeName=base_hair_back_r_capsule}]
 base_hair_front_r [{shapeType=Capsule, length=2, radius=2, shapeName=base_hair_front_r_capsule}]
 clavicle_l [{shapeType=Capsule, length=0.25, radius=10.5, shapeName=clavicle_l_capsule, center=<4.7161 -3.0378 -3.1531>, rotation=<rot 90 0 44.9996>}]
 clavicle_r [{shapeType=Capsule, length=2.5, radius=8.5, shapeName=clavicle_r_capsule, center=<-4.5027 -0.5761 -0.3115>, rotation=<rot 90 0 0>}]
 dyn_hair_back_01_c [{shapeType=Capsule, length=6, radius=2.5, shapeName=dyn_hair_back_01_c_capsule, center=<5.1769 0 0>, rotation=<rot 0 -90 90.0002>}]
 dyn_hair_back_01_l [{shapeType=Capsule, length=6.5, radius=2, shapeName=dyn_hair_back_01_l_capsule, center=<5.1769 0 0>, rotation=<rot 55 -89.9997 90.0004>}]
 dyn_hair_back_01_r [{shapeType=Capsule, length=6.5, radius=2, center=<5.1769 0 0>, rotation=<rot -50 -90.0002 90.0003>}]
 dyn_hair_back_02_c [{shapeType=Capsule, length=2.75, radius=1.75, center=<3.6105 0 0>, rotation=<rot -0.0005 0 90>}]
 dyn_hair_back_02_l [{
   shapeType=Capsule
   length=5.75
   radius=2
   shapeName=dyn_hair_back_02_l_capsule
   center=<4.2326 -0.0427 -0.1089>
   rotation=<rot -6.8972 14.7543 116.9251>
  }]
 dyn_hair_back_02_r [{
   shapeType=Capsule
   length=4.7875
   radius=1.7125
   shapeName=dyn_hair_back_02_r_capsule
   center=<5.1075 -0.6879 -1.5848>
   rotation=<rot -8.4373 -12.4483 76.3125>
  }]
 dyn_hair_back_03_c [{shapeType=Capsule, length=5, radius=2, shapeName=dyn_hair_back_03_c_capsule, center=<2.5859 0 -0.568>, rotation=<rot 0 0 90>}]
 dyn_hair_back_03_l [{
   shapeType=Capsule
   length=9
   radius=1.5
   shapeName=dyn_hair_back_03_l_capsule
   center=<2.765 0.9529 -1.0339>
   rotation=<rot -16.9604 19.2657 110.2958>
  }]
 dyn_hair_back_03_r [{
   shapeType=Capsule
   length=5
   radius=2
   shapeName=dyn_hair_back_03_r_capsule
   center=<4.0401 -1.8408 -0.3335>
   rotation=<rot -5.695 -3.194 83.3166>
  }]
 dyn_hair_back_04_c [{
   shapeType=Capsule
   length=11
   radius=1.75
   shapeName=dyn_hair_back_04_c_capsule
   center=<3.5859 0.0168 0.5514>
   rotation=<rot -0.0002 0 85.299>
  }]
 dyn_hair_back_04_l [{
   shapeType=Capsule
   length=6.75
   radius=1.75
   shapeName=dyn_hair_back_04_l_capsule
   center=<2.8697 0.0208 0.9405>
   rotation=<rot -0.0004 0 85.8598>
  }]
 dyn_hair_back_04_r [{shapeType=Capsule, length=4.5, radius=2.25, shapeName=dyn_hair_back_04_r_capsule, center=<3.1875 -1.5049 0.4952>, rotation=<rot 0 0 90>}]
 dyn_hair_back_05_c [{shapeType=Capsule, length=8.6714, radius=2.25, shapeName=dyn_hair_back_05_c_capsule, center=<5.4822 0 -0.6277>, rotation=<rot 79.9998 0 95>}]
 dyn_hair_front_01_r [{shapeType=Capsule, length=3, radius=1.5, shapeName=dyn_hair_front_01_r_capsule, center=<3 0 0>, rotation=<rot 90 0 0>}]
 dyn_hair_front_02_r [{shapeType=Capsule, length=3, radius=1.5, shapeName=dyn_hair_front_02_r_capsule, center=<3 0 0>, rotation=<rot 90 0 0>}]
 dyn_hair_front_03_r [{shapeType=Capsule, length=3.75, radius=1.5, shapeName=dyn_hair_front_03_r_capsule, center=<3.5351 0 0>, rotation=<rot 90 0 0>}]
 head [{shapeType=Capsule, length=3.75, radius=5.75, center=<3.7841 1.8995 0>, rotation=<rot 90 179.9998 180>}]
 neck_01 [
  {shapeType=Capsule, length=12.75, radius=4, shapeName=neck_01_capsule, center=<6.403 1.0662 0>, rotation=<rot 90 180 175.0005>}
  {shapeType=Capsule, length=21.25, radius=16.5, shapeName=neck_01_capsule, center=<6.7378 14.7178 0>}
 ]
 spine_05 [{shapeType=Capsule, length=10.75, radius=9.5, shapeName=spine_05_capsule, center=<4.9041 0.179 0>}]
}
constraints [
 {bone1Name=dyn_hair_back_02_c, bone2Name=dyn_hair_back_01_c, swing1LimitDegrees=5, swing2LimitDegrees=25, twistLimitDegrees=10}
 {
  bone1Name=dyn_hair_back_01_r
  bone2Name=base_hair_back_r
  swing1LimitDegrees=0
  swing1Motion=Locked
  swing2LimitDegrees=5
  twistLimitDegrees=0
  twistMotion=Locked
 }
 {
  bone1Name=head
  bone2Name=neck_01
  swing1LimitDegrees=45
  swing1Motion=Free
  swing2LimitDegrees=45
  swing2Motion=Free
  twistLimitDegrees=45
  twistMotion=Free
 }
 {
  bone1Name=neck_01
  bone2Name=spine_05
  swing1LimitDegrees=45
  swing1Motion=Free
  swing2LimitDegrees=45
  swing2Motion=Free
  twistLimitDegrees=45
  twistMotion=Free
 }
 {
  bone1Name=clavicle_l
  bone2Name=spine_05
  swing1LimitDegrees=45
  swing1Motion=Free
  swing2LimitDegrees=45
  swing2Motion=Free
  twistLimitDegrees=45
  twistMotion=Free
 }
 {
  bone1Name=clavicle_r
  bone2Name=spine_05
  swing1LimitDegrees=45
  swing1Motion=Free
  swing2LimitDegrees=45
  swing2Motion=Free
  twistLimitDegrees=45
  twistMotion=Free
 }
 {
  bone1Name=base_hair_back_l
  bone2Name=head
  swing1LimitDegrees=45
  swing1Motion=Free
  swing2LimitDegrees=45
  swing2Motion=Free
  twistLimitDegrees=45
  twistMotion=Free
 }
 {
  bone1Name=dyn_hair_back_01_l
  bone2Name=base_hair_back_l
  swing1LimitDegrees=0
  swing1Motion=Locked
  swing2LimitDegrees=5
  twistLimitDegrees=0
  twistMotion=Locked
 }
 {
  bone1Name=dyn_hair_back_02_l
  bone2Name=dyn_hair_back_01_l
  swing1LimitDegrees=5
  swing2LimitDegrees=15
  twistLimitDegrees=45
  twistMotion=Locked
 }
 {bone1Name=dyn_hair_back_03_l, bone2Name=dyn_hair_back_02_l, swing1LimitDegrees=10, swing2LimitDegrees=20, twistLimitDegrees=20}
 {
  bone1Name=dyn_hair_back_04_l
  bone2Name=dyn_hair_back_03_l
  swing1LimitDegrees=15
  swing2LimitDegrees=30
  twistLimitDegrees=45
  twistMotion=Locked
 }
 {
  bone1Name=base_hair_back_r
  bone2Name=head
  swing1LimitDegrees=45
  swing1Motion=Free
  swing2LimitDegrees=45
  swing2Motion=Free
  twistLimitDegrees=45
  twistMotion=Free
 }
 {
  bone1Name=dyn_hair_back_02_r
  bone2Name=dyn_hair_back_01_r
  swing1LimitDegrees=5
  swing2LimitDegrees=15
  twistLimitDegrees=45
  twistMotion=Locked
 }
 {
  bone1Name=dyn_hair_back_03_r
  bone2Name=dyn_hair_back_02_r
  swing1LimitDegrees=15
  swing2LimitDegrees=30
  twistLimitDegrees=45
  twistMotion=Locked
 }
 {
  bone1Name=dyn_hair_back_04_r
  bone2Name=dyn_hair_back_03_r
  swing1LimitDegrees=20
  swing2LimitDegrees=30
  twistLimitDegrees=45
  twistMotion=Locked
 }
 {
  bone1Name=base_hair_front_r
  bone2Name=head
  swing1LimitDegrees=45
  swing1Motion=Free
  swing2LimitDegrees=45
  swing2Motion=Free
  twistLimitDegrees=45
  twistMotion=Free
 }
 {
  bone1Name=dyn_hair_front_01_r
  bone2Name=base_hair_front_r
  swing1LimitDegrees=10
  swing2LimitDegrees=10
  twistLimitDegrees=45
  twistMotion=Locked
 }
 {
  bone1Name=dyn_hair_front_02_r
  bone2Name=dyn_hair_front_01_r
  swing1LimitDegrees=10
  swing2LimitDegrees=15
  twistLimitDegrees=45
  twistMotion=Locked
 }
 {
  bone1Name=dyn_hair_front_03_r
  bone2Name=dyn_hair_front_02_r
  swing1LimitDegrees=15
  swing2LimitDegrees=20
  twistLimitDegrees=45
  twistMotion=Locked
 }
 {bone1Name=dyn_hair_back_03_c, bone2Name=dyn_hair_back_02_c, swing1LimitDegrees=10, swing2LimitDegrees=30, twistLimitDegrees=10}
 {bone1Name=dyn_hair_back_04_c, bone2Name=dyn_hair_back_03_c, swing1LimitDegrees=10, swing2LimitDegrees=30, twistLimitDegrees=15}
 {
  bone1Name=dyn_hair_back_05_c
  bone2Name=dyn_hair_back_04_c
  swing1LimitDegrees=20
  swing2LimitDegrees=45
  twistLimitDegrees=45
  twistMotion=Locked
 }
 {
  bone1Name=base_hair_back_c
  bone2Name=head
  swing1LimitDegrees=45
  swing1Motion=Free
  swing2LimitDegrees=45
  swing2Motion=Free
  twistLimitDegrees=45
  twistMotion=Free
 }
 {
  bone1Name=dyn_hair_back_01_c
  bone2Name=base_hair_back_c
  swing1LimitDegrees=0
  swing1Motion=Locked
  swing2LimitDegrees=5
  twistLimitDegrees=0
  twistMotion=Locked
 }
]
