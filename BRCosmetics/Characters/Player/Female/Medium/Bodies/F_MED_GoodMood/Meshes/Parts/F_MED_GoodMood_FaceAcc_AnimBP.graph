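# Character-part AnimBP for the GoodMood female face accessory, animating hair and earrings over poses copied from parent and body meshes.
# Chained AnimDynamics swing earrings, locks, bangs and pin, then additive bone offsets correct back hair using spine-to-head bend during crouch or aiming, plus skydive-float state.
# Head-space RigidBody then simulates with either backpack or free physics asset, chosen by back bling/cape detection.
# State-driven gravity and motion scales tune that simulation.
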AnimBlueprint /BRCosmetics/Characters/Player/Female/Medium/Bodies/F_MED_GoodMood/Meshes/Parts/F_MED_GoodMood_FaceAcc_AnimBP
target_skeleton=&/F_MED_GoodMood_FaceAcc_Skeleton
graphs [
 AnimGraph {nodes [
   Output Pose(Result=Use cached pose 'BackpackPose'()) {layerGroup=None}
   n1 = Anim Dynamics - Bone: dyn_lock_01_l(ComponentPose=Local To Component(LocalPose=Use cached pose 'EaringCache'()), Alpha=1, ExternalForce="0,0,0") {
    angularSpringConstant=100000
    bAngularSpring=true
    alphaBoolBlend {blendOption=Linear}
    boundBone {boneName=dyn_lock_01_l}
    physicsBodyDefinitions [{
      boundBone {boneName=dyn_lock_01_l}
      boxExtents=<20 20 20>
      constraintSetup {angularLimitsMax=<0 5 0>, angularLimitsMin=<0 -30 0>, angularTarget=<0 0 0>}
      localJointOffset=<1 0 0>
     }]
   }
   n2 = Anim Dynamics - Bone: dyn_lock_02_l(ComponentPose=@n1, Alpha=1, ExternalForce="0,0,0") {
    angularSpringConstant=100000
    bAngularSpring=true
    alphaBoolBlend {blendOption=Linear}
    boundBone {boneName=dyn_lock_02_l}
    physicsBodyDefinitions [{
      boundBone {boneName=dyn_lock_02_l}
      boxExtents=<20 20 20>
      constraintSetup {angularLimitsMax=<0 5 0>, angularLimitsMin=<0 -30 0>, angularTarget=<0 0 0>}
      localJointOffset=<1 0 0>
     }]
   }
   n3 = Anim Dynamics - Bone: base_bang_c(ComponentPose=@n2, Alpha=1, ExternalForce="0,0,0") {
    angularSpringConstant=100000
    bAngularSpring=true
    alphaBoolBlend {blendOption=Linear}
    boundBone {boneName=base_bang_c}
    physicsBodyDefinitions [{
      boundBone {boneName=base_bang_c}
      boxExtents=<40 40 40>
      constraintSetup {angularLimitsMax=<0 5 5>, angularLimitsMin=<0 -15 -5>, angularTarget=<0 0 0>}
      localJointOffset=<5 0 0>
     }]
   }
   n4 = Anim Dynamics - Bone: dyn_pin_c(ComponentPose=@n3, Alpha=1, ExternalForce="0,0,0") {
    angularSpringConstant=1000
    bAngularSpring=true
    alphaBoolBlend {blendOption=Linear}
    boundBone {boneName=dyn_pin_c}
    physicsBodyDefinitions [{
      boundBone {boneName=dyn_pin_c}
      boxExtents=<40 40 40>
      constraintSetup {angularLimitsMax=<5 5 10>, angularLimitsMin=<-5 0 -10>, angularTarget=<0 0 0>}
      localJointOffset=<1 0 0>
     }]
   }
   n5 = Anim Dynamics - Bone: dyn_hair_front_01_c(ComponentPose=@n4, Alpha=1, ExternalForce="0,0,0") {
    angularSpringConstant=1000
    alphaBoolBlend {blendOption=Linear}
    boundBone {boneName=dyn_hair_front_01_c}
    physicsBodyDefinitions [{
      boundBone {boneName=dyn_hair_front_01_c}
      boxExtents=<40 40 40>
      constraintSetup {angularLimitsMax=<0 0 10>, angularLimitsMin=<0 -10 -10>, angularTarget=<0 0 0>}
      localJointOffset=<1 0 0>
     }]
   }
   n6 = Anim Dynamics - Bone: dyn_hair_front_02_c(ComponentPose=@n5, Alpha=1, ExternalForce="0,0,0") {
    angularSpringConstant=1000
    alphaBoolBlend {blendOption=Linear}
    boundBone {boneName=dyn_hair_front_02_c}
    physicsBodyDefinitions [{
      boundBone {boneName=dyn_hair_front_02_c}
      boxExtents=<35 35 35>
      constraintSetup {angularLimitsMax=<0 0 10>, angularLimitsMin=<0 -30 -10>, angularTarget=<0 0 0>}
      localJointOffset=<1 0 0>
     }]
   }
   n7 = Anim Dynamics - Bone: dyn_hair_front_02_l(ComponentPose=@n6, Alpha=1, ExternalForce="0,0,0") {
    angularSpringConstant=100
    bAngularSpring=true
    alphaBoolBlend {blendOption=Linear}
    boundBone {boneName=dyn_hair_front_02_l}
    physicsBodyDefinitions [{
      boundBone {boneName=dyn_hair_front_02_l}
      boxExtents=<40 40 40>
      constraintSetup {angularLimitsMax=<0 0 10>, angularLimitsMin=<0 -10 -10>, angularTarget=<0 0 0>}
      localJointOffset=<5 0 0>
     }]
   }
   n8 = Anim Dynamics - Bone: dyn_hair_front_03_l(ComponentPose=@n7, Alpha=1, ExternalForce="0,0,0") {
    angularSpringConstant=1000
    bAngularSpring=true
    alphaBoolBlend {blendOption=Linear}
    boundBone {boneName=dyn_hair_front_03_l}
    physicsBodyDefinitions [{
      boundBone {boneName=dyn_hair_front_03_l}
      boxExtents=<35 35 35>
      constraintSetup {angularLimitsMax=<0 0 10>, angularLimitsMin=<0 -30 -20>, angularTarget=<0 0 0>}
      localJointOffset=<5 0 0>
     }]
   }
   Save cached pose 'LockCache'(Pose=Component To Local(ComponentPose=@n8)) {cacheName=LockCache}
   n9 = "Transform (Modify) Bone - Bone: base_hair_back_l"!(
    ComponentPose=Local To Component(LocalPose=Use cached pose 'FixCache'())
    bAlphaBoolEnabled:bool=Get bIsSkydiveFloating()
    Translation="0.000,0.000,0.371"
    Rotation="0.000,5.000,20.000"
    Scale="1,1,1"
   ) -> (Pose) {
    alphaInputType=Bool
    rotationMode=BMM_Additive
    rotationSpace=BCS_BoneSpace
    translationMode=BMM_Additive
    translationSpace=BCS_BoneSpace
    alphaBoolBlend {blendOption=QuarticInOut, blendOutTime=1}
    boneToModify {boneName=base_hair_back_l}
    rotation=<rot 0 5 20>
    translation {z=0.371}
   }
   n10 = Get Linear Scale Backpack() -> (Linear Scale Backpack:Vector)
   n11 = RigidBody!(
    ComponentPose=@n9
    Alpha
    OverrideWorldGravity:Vector=Get GravityOverride()
    ExternalForce
    ComponentLinearAccScale:Vector=@n10
    ComponentLinearVelScale:Vector=@n10
   ) -> (Pose) {
    bClampLinearTranslationLimitToRefPose=true
    bOverrideWorldGravity=true
    overridePhysicsAsset=&/F_MED_GoodMood_FaceAcc_BackPack_Physics
    simulationSpace=BaseBoneSpace
    alphaBoolBlend {blendOption=Linear}
    baseBoneRef {boneName=head}
    componentLinearAccScale=<0.5 0.3 0.3>
    componentLinearVelScale=<0.5 0.3 0.3>
   }
   n12 = Get Linear Scale Body() -> (Linear Scale Body:Vector)
   n13 = RigidBody(
    ComponentPose=Local To Component(LocalPose=Use cached pose 'FixCache'())
    OverrideWorldGravity=Get GravityOverride()
    ComponentLinearAccScale=@n12
    ComponentLinearVelScale=@n12
   ) {
    bClampLinearTranslationLimitToRefPose=true
    bOverrideWorldGravity=true
    overridePhysicsAsset=&/F_MED_GoodMood_FaceAcc_Free_Physics
    simulationSpace=BaseBoneSpace
    alphaBoolBlend {blendOption=Linear}
    baseBoneRef {boneName=head}
    componentLinearAccScale=<0.5 0.3 0.3>
    componentLinearVelScale=<0.5 0.3 0.3>
   }
   Save cached pose 'BackpackPose'(Pose=Blend Poses by bool(
     BlendPose_0=Component To Local(ComponentPose=@n11)
     BlendPose_1=Component To Local(ComponentPose=@n13)
     bActiveValue=Get IsWearing() -> (IsWearing:bool)
    ) {blendType=Linear}) {cacheName=BackpackPose}
   n14 = Layered blend per bone(
    BasePose=Copy Pose From Mesh() {bCopyCurves=true, bUseAttachedParent=true}
    BlendPoses_0=Copy Pose From Mesh(SourceMeshComponent=Get Body_Mesh())
   ) {layerSetup [{branchFilters [{boneName=clavicle_l}, {boneName=clavicle_r}, {boneName=upperarm_r}, {boneName=upperarm_l}]}]}
   Save cached pose 'StartPose'(Pose=@n14) {cacheName=StartPose}
   n15 = Anim Dynamics - Bone: Dyn_Earing_01_l(ComponentPose=Local To Component(LocalPose=Use cached pose 'StartPose'()), Alpha=1, ExternalForce="0,0,0") {
    angularSpringConstant=300
    bAngularSpring=true
    bUseSphericalLimits=true
    alphaBoolBlend {blendOption=Linear}
    boundBone {boneName=Dyn_Earing_01_l}
    componentLinearAccScale=<0.1 0.1 0.1>
    componentLinearVelScale=<0.1 0.1 0.1>
    physicsBodyDefinitions [{
      boundBone {boneName=Dyn_Earing_01_l}
      boxExtents=<30 30 20>
      constraintSetup {angularLimitsMax=<0 40 0>, angularLimitsMin=<0 -40 0>, angularTarget=<1 0 0>}
      localJointOffset=<2 0 0>
     }]
   }
   n16 = Anim Dynamics - Bone: Dyn_Earing_02_l(ComponentPose=@n15, Alpha=1, ExternalForce="0,0,0") {
    angularSpringConstant=10000
    bAngularSpring=true
    bUseSphericalLimits=true
    alphaBoolBlend {blendOption=Linear}
    boundBone {boneName=Dyn_Earing_02_l}
    componentLinearAccScale=<0.1 0.1 0.1>
    componentLinearVelScale=<0.1 0.1 0.1>
    physicsBodyDefinitions [{
      boundBone {boneName=Dyn_Earing_02_l}
      boxExtents=<40 40 20>
      constraintSetup {angularLimitsMax=<0 10 0>, angularLimitsMin=<0 -10 -10>, angularTarget=<1 0 0>}
      localJointOffset=<2 0 0>
     }]
   }
   n17 = Anim Dynamics - Bone: Dyn_Earing_01_r(ComponentPose=@n16, Alpha=1, ExternalForce="0,0,0") {
    angularSpringConstant=300
    bAngularSpring=true
    bUseSphericalLimits=true
    alphaBoolBlend {blendOption=Linear}
    boundBone {boneName=Dyn_Earing_01_r}
    componentLinearAccScale=<0.1 0.1 0.1>
    componentLinearVelScale=<0.1 0.1 0.1>
    physicsBodyDefinitions [{
      boundBone {boneName=Dyn_Earing_01_r}
      boxExtents=<30 30 20>
      constraintSetup {angularLimitsMax=<0 40 0>, angularLimitsMin=<0 -40 0>, angularTarget=<1 0 0>}
      localJointOffset=<2 0 0>
     }]
   }
   n18 = Anim Dynamics - Bone: Dyn_Earing_02_r(ComponentPose=@n17, Alpha=1, ExternalForce="0,0,0") {
    angularSpringConstant=10000
    bAngularSpring=true
    bUseSphericalLimits=true
    alphaBoolBlend {blendOption=Linear}
    boundBone {boneName=Dyn_Earing_02_r}
    componentLinearAccScale=<0.1 0.1 0.1>
    componentLinearVelScale=<0.1 0.1 0.1>
    physicsBodyDefinitions [{
      boundBone {boneName=Dyn_Earing_02_r}
      boxExtents=<40 40 20>
      constraintSetup {angularLimitsMax=<0 10 0>, angularLimitsMin=<0 -10 -10>, angularTarget=<1 0 0>}
      localJointOffset=<2 0 0>
     }]
   }
   Save cached pose 'EaringCache'(Pose=Component To Local(ComponentPose=@n18)) {cacheName=EaringCache}
   n19 = Get bIsSkydiveFloating()
   n20 = Get HeadUpWCrouch() -> (HeadUpWCrouch:double)
   n21 = "Transform (Modify) Bone - Bone: base_hair_back_r"(
    ComponentPose=Local To Component(LocalPose=Use cached pose 'LockCache'())
    Alpha=@n20
    Translation="3.745,-6.919,-3.191"
    Rotation="20,40,10"
    Scale="1,1,1"
   ) {
    rotationMode=BMM_Additive
    rotationSpace=BCS_BoneSpace
    translationMode=BMM_Additive
    translationSpace=BCS_BoneSpace
    alphaBoolBlend {blendOption=QuarticInOut, blendOutTime=1}
    boneToModify {boneName=base_hair_back_r}
    rotation=<rot 20 40 10>
    translation=<3.745 -6.919 -3.191>
   }
   n22 = "Transform (Modify) Bone - Bone: base_hair_back_l"!(ComponentPose=@n21, Alpha:float=@n20, Translation="0.496,2.431,3.991", Rotation="40,20.811,22.479", Scale="1,1,1") -> (Pose) {
    rotationMode=BMM_Additive
    rotationSpace=BCS_BoneSpace
    translationMode=BMM_Additive
    translationSpace=BCS_BoneSpace
    alphaBoolBlend {blendOption=QuarticInOut, blendOutTime=1}
    boneToModify {boneName=base_hair_back_l}
    rotation=<rot 40 20.811 22.479>
    translation=<0.496 2.431 3.991>
   }
   n23 = "Transform (Modify) Bone - Bone: base_hair_back_c"(ComponentPose=@n22, Alpha=@n20, Translation="0.373,-0.423,-1.765", Rotation="58.301,-0.023,-0.061", Scale="1,1,1") {
    rotationMode=BMM_Additive
    rotationSpace=BCS_BoneSpace
    translationMode=BMM_Additive
    translationSpace=BCS_BoneSpace
    alphaBoolBlend {blendOption=QuarticInOut, blendOutTime=1}
    boneToModify {boneName=base_hair_back_c}
    rotation=<rot 58.301 -0.023 -0.061>
    translation=<0.373 -0.423 -1.765>
   }
   n24 = Get HeadUpAim() -> (HeadUpAim:double)
   n25 = "Transform (Modify) Bone - Bone: base_hair_back_c"(ComponentPose=@n23, Alpha=@n24, Translation="0,0,0", Rotation="0,0,0", Scale="1,1,1") {
    rotationMode=BMM_Additive
    rotationSpace=BCS_BoneSpace
    translationMode=BMM_Additive
    translationSpace=BCS_BoneSpace
    alphaBoolBlend {blendOption=Linear}
    alphaScaleBiasClamp {bInterpResult=true, interpSpeedIncreasing=1}
    boneToModify {boneName=base_hair_back_c}
   }
   n26 = "Transform (Modify) Bone - Bone: base_hair_back_r"(ComponentPose=@n25, Alpha=@n24, Translation="-5.036,-0.143,11.538", Rotation="28.284,-11.091,64.504", Scale="1,1,1") {
    rotationMode=BMM_Additive
    rotationSpace=BCS_BoneSpace
    translationMode=BMM_Additive
    translationSpace=BCS_BoneSpace
    alphaBoolBlend {blendOption=Linear}
    alphaScaleBiasClamp {bInterpResult=true}
    boneToModify {boneName=base_hair_back_r}
    rotation=<rot 28.2839 -11.091 64.504>
    translation=<-5.036 -0.143 11.538>
   }
   n27 = Get HeadOnGround() -> (HeadOnGround:double)
   n28 = "Transform (Modify) Bone - Bone: base_hair_back_r"(ComponentPose=@n26, Alpha=@n27, Translation="0,-2,0", Rotation="0,0,50", Scale="1,1,1") {
    rotationMode=BMM_Additive
    rotationSpace=BCS_BoneSpace
    translationMode=BMM_Additive
    translationSpace=BCS_BoneSpace
    alphaBoolBlend {blendOption=QuarticInOut, blendOutTime=1}
    boneToModify {boneName=base_hair_back_r}
    rotation {roll=50}
    translation {y=-2}
   }
   n29 = "Transform (Modify) Bone - Bone: base_hair_back_c"(ComponentPose=@n28, Alpha=@n27, Translation="0,-2,0", Rotation="-15,0,50", Scale="1,1,1") {
    rotationMode=BMM_Additive
    rotationSpace=BCS_BoneSpace
    translationMode=BMM_Additive
    translationSpace=BCS_BoneSpace
    alphaBoolBlend {blendOption=QuarticInOut, blendOutTime=1}
    boneToModify {boneName=base_hair_back_c}
    rotation {pitch=-15, roll=50}
    translation {y=-2}
   }
   n30 = "Transform (Modify) Bone - Bone: base_hair_back_l"(ComponentPose=@n29, Alpha=@n27, Translation="0,-2,0", Rotation="0,0,40", Scale="1,1,1") {
    rotationMode=BMM_Additive
    rotationSpace=BCS_BoneSpace
    translationMode=BMM_Additive
    translationSpace=BCS_BoneSpace
    alphaBoolBlend {blendOption=QuarticInOut, blendOutTime=1}
    boneToModify {boneName=base_hair_back_l}
    rotation {roll=40}
    translation {y=-2}
   }
   n31 = "Transform (Modify) Bone - Bone: base_hair_back_l"!(ComponentPose=@n30, bAlphaBoolEnabled:bool=@n19, Translation="0,-1,-1", Rotation="39.999,0.359,0.288", Scale="1,1,1") -> (Pose) {
    alphaInputType=Bool
    rotationMode=BMM_Additive
    rotationSpace=BCS_BoneSpace
    translationMode=BMM_Additive
    translationSpace=BCS_BoneSpace
    alphaBoolBlend {blendOption=QuarticInOut, blendOutTime=1}
    boneToModify {boneName=base_hair_back_l}
    rotation=<rot 39.999 0.359 0.288>
    translation {y=-1, z=-1}
   }
   n32 = "Transform (Modify) Bone - Bone: base_hair_back_c"!(ComponentPose=@n31, bAlphaBoolEnabled:bool=@n19, Translation="0,0,0", Rotation="0.084,0.017,19.999", Scale="1,1,1") -> (Pose) {
    alphaInputType=Bool
    rotationMode=BMM_Additive
    rotationSpace=BCS_BoneSpace
    translationSpace=BCS_BoneSpace
    alphaBoolBlend {blendOption=QuarticInOut, blendOutTime=1}
    boneToModify {boneName=base_hair_back_c}
    rotation=<rot 0.084 0.017 19.999>
   }
   n33 = "Transform (Modify) Bone - Bone: base_hair_back_r"!(ComponentPose=@n32, bAlphaBoolEnabled:bool=@n19, Translation="1,-1,1", Rotation="44.939,2.418,2.511", Scale="1,1,1") -> (Pose) {
    alphaInputType=Bool
    rotationMode=BMM_Additive
    rotationSpace=BCS_BoneSpace
    translationMode=BMM_Additive
    translationSpace=BCS_BoneSpace
    alphaBoolBlend {blendOption=QuarticInOut, blendOutTime=1}
    boneToModify {boneName=base_hair_back_r}
    rotation=<rot 44.939 2.418 2.511>
    translation=<1 -1 1>
   }
   n34 = "Transform (Modify) Bone - Bone: head_fix"(ComponentPose=@n33, Alpha=1, Translation="-5,2,3", Rotation="-15,0,0", Scale="1,1,1") {
    rotationMode=BMM_Additive
    rotationSpace=BCS_BoneSpace
    translationMode=BMM_Additive
    translationSpace=BCS_BoneSpace
    alphaBoolBlend {blendOption=QuarticInOut, blendOutTime=1}
    boneToModify {boneName=head_fix}
    rotation {pitch=-15}
    translation=<-5 2 3>
   }
   Save cached pose 'FixCache'(Pose=Component To Local(ComponentPose=@n34)) {cacheName=FixCache}
  ]}
 EventGraph {nodes [
   Event OnCharacterCustomizationLoaded()
   > n1 = Cast To PlayerPawn_Athena(Object=TryGetPawnOwner())
   > Set Body_Mesh(Body_Mesh=GetSkeletalMeshForPartType(self=@n1))
   n2 = CheckForBackbling()
   > n3 = Set isWearingBackbling(isWearingBackbling=AND Boolean!(A=@n2, B=NOT Boolean(A=@n2.bIsBackblingACape), C:bool=NOT Boolean(A=@n2.bIsBackblingLong)) -> (ReturnValue))
   > n4 = Set Is Wearing Cape(Is Wearing Cape=OR Boolean(A=@n2.bIsBackblingACape, B=@n2.bIsBackblingLong))
   > Set IsWearing(IsWearing:bool=OR Boolean(A=@n4, B=@n3)) -> (Output_Get:bool)
   Event BlueprintInitializeAnimation()
   > Set Skel Mesh(Skel Mesh=GetOwnerSkeletalMeshComponent())
   Event BlueprintUpdateAnimation()
   > n5 = Sequence!(execute) -> (then_0, then_1, then_2:Exec, then_3:Exec, then_4:Exec)
   n6 = Branch(execute=@n5.then_1, Condition=Get bIsSkydiveFloating())
   > Set Linear Scale Body(Linear Scale Body:Vector="0.3,0.2,0.2") -> (Output_Get:Vector)
   > Set Linear Scale Backpack(Linear Scale Backpack:Vector="0.1,0,0") -> (Output_Get:Vector)
   Set Linear Scale Body(execute=@n6.else, Linear Scale Body="0.35,0.3,0.3")
   > Set Linear Scale Backpack(Linear Scale Backpack="0.2,0.1,0.1")
   n7 = Get Skel Mesh()
   n8 = Set Hair Up(
    execute=@n5.then_2
    Hair Up:double=float * float(
     A=GetAlphaFromReferenceTransformDelta(SkeletalMeshComponent=@n7, BoneName=spine_05, BaseName=head, InputRange=(X=0,Y=90), TwistAxis="0,0,1")
     B=1
    ) {is=title K2Node_PromotableOperator}
   ) -> (Output_Get:double)
   > n9 = Branch(Condition=Get bIsCrouching())
   > Set HeadUpWCrouch(HeadUpWCrouch:double=@n8) -> (Output_Get:double)
   Set HeadUpWCrouch(execute=@n9.else, HeadUpWCrouch=0)
   n10 = NOT Boolean(A=Get bIsCrouching())
   n11 = Get bIsOnGround() -> (bIsOnGround:bool)
   n12 = Get bIsTargeting()
   n13 = Get bIsTargetingConsumableThrow() -> (bIsTargetingConsumableThrow:bool)
   n14 = Branch(execute=@n5.then_3, Condition=AND Boolean!(A=AND Boolean(A=@n11, B=@n10), B=AND Boolean(A=@n12, B=@n13)) -> (ReturnValue))
   > Set HeadOnGround(HeadOnGround:double=@n8) -> (Output_Get:double)
   Set HeadOnGround(execute=@n14.else, HeadOnGround=0)
   n15 = Gravity Override State Function DEPRECATED(
    execute=@n5.then_0
    CharacterPartAnimInstance=Self-Reference()
    GameTimeInSeconds=GetGameTimeInSeconds()
    GravityOverrideParams=Get Gravity Override Params()
   )
   > Set GravityOverride(GravityOverride=@n15)
   n16 = Branch(execute=@n5.then_4, Condition=Get bIsTargeting())
   > Set HeadUpAim(HeadUpAim:double=Get Hair Up() -> (Hair Up:double)) -> (Output_Get:double)
   Set HeadUpAim(execute=@n16.else, HeadUpAim=0)
  ]}
]
variables [
 Body_Mesh {type=SkeletalMeshComponent}
 Gravity Override Params {type=GravityOverrideParamsStruct, default {skydive_FloatingDown {gravityOverride=<0 -700 300>}}}
 GravityOverride {type=Vector}
 Hair Up {type=number}
 HeadOnGround {type=number}
 HeadUpAim {type=number}
 HeadUpWCrouch {type=number}
 Is Wearing Cape {type=boolean}
 IsWearing {type=boolean}
 isWearingBackbling {type=boolean}
 Linear Scale Backpack {type=Vector}
 Linear Scale Body {type=Vector}
 Skel Mesh {type=SkeletalMeshComponent}
]
